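# Material function giving one octave of 2D value noise, exposed as "Noise for glitch fbm" and stored with the ByteGear character's Roll skin materials.
# Floor and Frac split coordinates into cell ID and local position, with MF_PseudoRandom hashing four neighboring cell corners.
# Bilinear lerps blend the corners using a SmoothStep weight whose offset upper edge appears driven by the blockiness input.
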MaterialFunction /BRCosmetics/Characters/Player/Male/Medium/Bodies/M_MED_ByteGear/Skins/Roll/Materials/MF_NoiseForGlitchFbm
graph {
 properties {bExposeToLibrary=true, description=Noise for glitch fbm, userExposedCaption=Noise for glitch fbm}
 nodes [
  n1 = NamedRerouteUsage()
  n2 = NamedRerouteUsage()
  n3 = ComponentMask(None=NamedRerouteUsage()) {r=true}
  n4 = NamedRerouteUsage()
  n5 = NamedRerouteUsage()
  n6 = ComponentMask(None=NamedRerouteUsage()) {r=true}
  n7 = NamedRerouteUsage()
  FunctionOutput(None=LinearInterpolate(A=LinearInterpolate(A=@n1, B=@n2, Alpha=@n3), B=LinearInterpolate(A=@n4, B=@n5, Alpha=@n6), Alpha=ComponentMask(None=@n7) {g=true}))
  n8 = TextureCoordinate()
  NamedRerouteDeclaration(None=FunctionInput(Preview=Multiply(A=@n8) {constB=5}) {inputName=uv, inputType=FunctionInput_Vector2, previewValue=<222 1234 0 1>}) {name=uv, nodeColor=<rgba 0 0.9766 1 1>}
  NamedRerouteDeclaration(None=FunctionInput(Preview=Constant() {r=3}) {inputName=blockiness, inputType=FunctionInput_Scalar, sortPriority=1, previewValue=<1 0 0 1>}) {name=blockiness, nodeColor=<rgba 1 0.7031 0 1>}
  NamedRerouteDeclaration(None=Frac(None=NamedRerouteUsage())) {name=lv, nodeColor=<rgba 0.3828 0 1 1>}
  NamedRerouteDeclaration(None=Floor(None=NamedRerouteUsage())) {name=id, nodeColor=<rgba 1 0 0.2578 1>}
  NamedRerouteDeclaration(None=MaterialFunctionCall!(seed=NamedRerouteUsage()) {materialFunction=&/MF_PseudoRandom}.Result) {name=n1, nodeColor=<rgba 0 0.5781 1 1>}
  n9 = NamedRerouteUsage()
  n10 = Constant2Vector() {r=1}
  NamedRerouteDeclaration(None=MaterialFunctionCall(seed=Add(A=@n9, B=@n10)) {materialFunction=&MF_PseudoRandom}.Result) {name=n2, nodeColor=<rgba 0.8984 1 0 1>}
  n11 = NamedRerouteUsage()
  n12 = Constant2Vector() {g=1}
  NamedRerouteDeclaration(None=MaterialFunctionCall(seed=Add(A=@n11, B=@n12)) {materialFunction=&MF_PseudoRandom}.Result) {name=n3, nodeColor=<rgba 0.7813 0 1 1>}
  n13 = NamedRerouteUsage()
  n14 = Constant2Vector() {g=1, r=1}
  NamedRerouteDeclaration(None=MaterialFunctionCall(seed=Add(A=@n13, B=@n14)) {materialFunction=&MF_PseudoRandom}.Result) {name=n4, nodeColor=<rgba 0 1 0.4609 1>}
  n15 = NamedRerouteUsage()
  NamedRerouteDeclaration(None=SmoothStep(Max=Add(A=@n15), Value=NamedRerouteUsage())) {name=u, nodeColor=<rgba 1 0.1406 0 1>}
 ]
}
