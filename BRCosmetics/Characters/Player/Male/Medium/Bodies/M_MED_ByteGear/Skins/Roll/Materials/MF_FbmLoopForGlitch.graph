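# One fractal-noise (fBm) octave step for a glitch effect, stored with the ByteGear Roll skin materials.
# It adds amplitude-scaled MF_NoiseForGlitchFbm noise to the running value, halves amplitude and decrements count.
# UVs are scaled by complexity and shifted using MF_PseudoRandom and floored time, giving stepped jumps.
# Outputs mirror inputs through named reroutes, so copies can chain as an unrolled loop.
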MaterialFunction /BRCosmetics/Characters/Player/Male/Medium/Bodies/M_MED_ByteGear/Skins/Roll/Materials/MF_FbmLoopForGlitch
graph {nodes [
  FunctionOutput(None=NamedRerouteUsage()) {outputName=value}
  FunctionOutput(None=NamedRerouteUsage()) {outputName=amplitude, sortPriority=1}
  FunctionOutput(None=NamedRerouteUsage()) {outputName=uv, sortPriority=2}
  FunctionOutput(None=NamedRerouteUsage()) {outputName=blockiness, sortPriority=3}
  FunctionOutput(None=NamedRerouteUsage()) {outputName=complexity, sortPriority=4}
  FunctionOutput(None=NamedRerouteUsage()) {outputName=count, sortPriority=5}
  FunctionOutput(None=NamedRerouteUsage()) {outputName=timemultiplier, sortPriority=6}
  NamedRerouteDeclaration(None=FunctionInput(Preview=Time() {bOverride_Period=true, period=3}) {inputName=value, inputType=FunctionInput_Scalar}) {name=value, nodeColor=<rgba 0.7344 1 0 1>}
  NamedRerouteDeclaration(None=FunctionInput(Preview=Constant() {r=0.5}) {inputName=amplitude, inputType=FunctionInput_Scalar, sortPriority=1}) {name=amplitude, nodeColor=<rgba 0.9453 0 1 1>}
  NamedRerouteDeclaration(None=Subtract(A=NamedRerouteUsage())) {name=count_update, nodeColor=<rgba 0.9658 1 0.0673 1>}
  NamedRerouteDeclaration(None=FunctionInput(Preview=Constant() {r=20}) {inputName=timemultiplier, sortPriority=6}) {name=timemultiplier, nodeColor=<rgba 0 1 0.8238 1>}
  NamedRerouteDeclaration(None=Multiply(A=NamedRerouteUsage(), B=Constant() {r=1})) {name=timemultupdated, nodeColor=<rgba 0.3594 0 1 1>}
  NamedRerouteDeclaration(None=FunctionInput(Preview=TextureCoordinate()) {inputName=uv, inputType=FunctionInput_Scalar, sortPriority=2}) {name=uv, nodeColor=<rgba 0 1 0.625 1>}
  NamedRerouteDeclaration(None=FunctionInput(Preview=Constant() {r=3}) {inputName=blockiness, inputType=FunctionInput_Scalar, sortPriority=3}) {name=blockiness, nodeColor=<rgba 1 0.5391 0 1>}
  NamedRerouteDeclaration(None=FunctionInput(Preview=Constant() {r=3}) {inputName=count, inputType=FunctionInput_Scalar, sortPriority=5}) {name=count, nodeColor=<rgba 0.2188 0 1 1>}
  n1 = NamedRerouteUsage()
  n2 = Divide(A=Ceil(None=Multiply(A=@n1) {constB=2}))
  n3 = Floor(None=Multiply(A=Time() {period=12}, B=NamedRerouteUsage()) {constB=100})
  n4 = Add(
   A=Multiply(A=MaterialFunctionCall!(seed=@n2) {materialFunction=&/MF_PseudoRandom}.Result) {constB=2}
   B=Divide(A=Divide(A=@n3) {constB=5}, B=NamedRerouteUsage())
  )
  NamedRerouteDeclaration(None=Add(A=Subtract(A=@n4), B=NamedRerouteUsage())) {name=uv_updated, nodeColor=<rgba 0.1016 1 0 1>}
  n5 = NamedRerouteUsage()
  n6 = MaterialFunctionCall!(uv=NamedRerouteUsage(), blockiness=NamedRerouteUsage()) {materialFunction=&/MF_NoiseForGlitchFbm}
  NamedRerouteDeclaration(None=Add(A=Multiply(A=@n5, B=@n6.Result), B=NamedRerouteUsage())) {name=value_updated, nodeColor=<rgba 1 0 0.4219 1>}
  NamedRerouteDeclaration(None=Multiply(A=NamedRerouteUsage()) {constB=0.5}) {name=amplitude_updated, nodeColor=<rgba 0 0.7422 1 1>}
  NamedRerouteDeclaration(None=FunctionInput(Preview=Constant() {r=1.5}) {inputName=complexity, inputType=FunctionInput_Scalar, sortPriority=4}) {name=complexity, nodeColor=<rgba 0 0.3133 1 1>}
  NamedRerouteDeclaration(None=Multiply(A=NamedRerouteUsage(), B=NamedRerouteUsage())) {name=uv_complexity, nodeColor=<rgba 0.6172 0 1 1>}
  NamedRerouteUsage()
 ]}
