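# Material function that builds fractal Brownian motion (fBm) noise for a glitch effect in the ByteGear Roll skin materials.
# It unrolls the fBm loop as three chained MF_FbmLoopForGlitch calls, each feeding the next, and outputs the last value.
# Inputs for UV, blockiness, complexity and time multiplier are shared through named reroutes, alongside constant starting value and amplitude.
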MaterialFunction /BRCosmetics/Characters/Player/Male/Medium/Bodies/M_MED_ByteGear/Skins/Roll/Materials/MF_FbmForGlitch
graph {nodes [
  n1 = MaterialFunctionCall!(
   value=NamedRerouteUsage()
   amplitude=NamedRerouteUsage()
   uv=NamedRerouteUsage()
   blockiness=NamedRerouteUsage()
   complexity=NamedRerouteUsage()
   count=Constant() {r=3}
   timemultiplier=NamedRerouteUsage()
  ) {materialFunction=&/MF_FbmLoopForGlitch}
  n2 = MaterialFunctionCall(value=@n1.value, amplitude=@n1.value, uv=@n1.value, blockiness=@n1.value, complexity=@n1.value, count=@n1.value, timemultiplier=@n1.value) {materialFunction=&MF_FbmLoopForGlitch}
  FunctionOutput(None=MaterialFunctionCall(value=@n2.value, amplitude=@n2.value, uv=@n2.value, blockiness=@n2.value, complexity=@n2.value, count=@n2.value, timemultiplier=@n2.value) {materialFunction=&MF_FbmLoopForGlitch}.value) {outputName=value}
  NamedRerouteDeclaration(None=FunctionInput(Preview=TextureCoordinate()) {bUsePreviewValueAsDefault=true, inputName=uv, inputType=FunctionInput_Scalar}) {name=uv, nodeColor=<rgba 0.4531 0 1 1>}
  NamedRerouteDeclaration(None=FunctionInput(Preview=Constant() {r=20}) {inputName=timemultiplier, sortPriority=4}) {name=timemultiplier, nodeColor=<rgba 1 0.6797 0 1>}
  NamedRerouteDeclaration(None=FunctionInput(Preview=Constant() {r=3}) {bUsePreviewValueAsDefault=true, inputName=blockiness, inputType=FunctionInput_Scalar, sortPriority=2}) {name=blockiness, nodeColor=<rgba 1 0 0.1875 1>}
  NamedRerouteDeclaration(None=FunctionInput(Preview=Constant() {r=1}) {bUsePreviewValueAsDefault=true, inputName=complexity, inputType=FunctionInput_Scalar, sortPriority=3}) {name=complexity, nodeColor=<rgba 0 0.5078 1 1>}
  NamedRerouteDeclaration(None=Constant()) {name=starting_value, nodeColor=<rgba 0.8281 1 0 1>}
  NamedRerouteDeclaration(None=Constant() {r=0.5}) {name=starting_amplitude, nodeColor=<rgba 0.8516 0 1 1>}
 ]}
