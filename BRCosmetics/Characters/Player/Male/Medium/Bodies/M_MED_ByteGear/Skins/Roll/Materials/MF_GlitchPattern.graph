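# Material function for the ByteGear Roll skin that outputs an animated RGB glitch pattern as separate red, green, blue and alpha channels.
# MF_FbmForGlitch generates time-driven noise at differing blockiness, complexity and speed, thresholded by per-color low-value inputs.
# Horizontal UV adds, subtracts and multiplies create per-channel coordinates for sampling the GlitchTexture parameter, giving color-split distortion.
# Each output multiplies a value by a smoothstep mask, with wiring carried through named reroutes.
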MaterialFunction /BRCosmetics/Characters/Player/Male/Medium/Bodies/M_MED_ByteGear/Skins/Roll/Materials/MF_GlitchPattern
graph {nodes [
  Constant() {r=90}
  n1 = NamedRerouteUsage()
  n2 = NamedRerouteUsage()
  n3 = NamedRerouteUsage()
  FunctionOutput(None=Multiply(A=NamedRerouteUsage(), B=SmoothStep(Min=@n1, Max=@n2, Value=@n3) {constMax=0.5, constMin=0.2})) {outputName=RedChannel}
  n4 = NamedRerouteUsage()
  n5 = NamedRerouteUsage()
  n6 = NamedRerouteUsage()
  FunctionOutput(None=Multiply(A=NamedRerouteUsage(), B=SmoothStep(Min=@n4, Max=@n5, Value=@n6) {constMax=0.5, constMin=0.2})) {outputName=GreenChannel, sortPriority=1}
  n7 = NamedRerouteUsage()
  n8 = NamedRerouteUsage()
  n9 = NamedRerouteUsage()
  FunctionOutput(None=Multiply(A=NamedRerouteUsage(), B=SmoothStep(Min=@n7, Max=@n8, Value=@n9) {constMax=0.5, constMin=0.2})) {outputName=BlueChannel, sortPriority=2}
  n10 = NamedRerouteUsage()
  n11 = NamedRerouteUsage()
  n12 = NamedRerouteUsage()
  FunctionOutput(None=Multiply(A=NamedRerouteUsage(), B=SmoothStep(Min=@n10, Max=@n11, Value=@n12) {constMax=0.5, constMin=0.2})) {outputName=AlphaChannel, sortPriority=3}
  NamedRerouteDeclaration(None=Multiply(A=NamedRerouteUsage())) {name=uv_updated, nodeColor=<rgba 1 0 0.9141 1>}
  NamedRerouteDeclaration(None=NamedRerouteUsage()) {name=blue_UVs, nodeColor=<rgba 0.1466 0.2061 1 1>}
  n13 = NamedRerouteUsage()
  n14 = Constant() {r=1}
  n15 = Constant() {r=1.5}
  n16 = Add(A=NamedRerouteUsage()) {constB=6.7}
  NamedRerouteDeclaration(None=SmoothStep(
    Min=NamedRerouteUsage()
    Value=MaterialFunctionCall!(uv=@n13, blockiness=@n14, complexity=@n15, timemultiplier=@n16) {materialFunction=&/MF_FbmForGlitch}.value
   ) {constMin=0.5}) {name=noisevalgreen, nodeColor=<rgba 0.0461 1 0.1353 1>}
  NamedRerouteDeclaration(None=TextureSample(UVs=NamedRerouteUsage(), Tex=NamedRerouteUsage()) {samplerSource=SSM_Wrap_WorldGroupSettings}.G) {name=green_channel, nodeColor=<rgba 0.108 1 0.1851 1>}
  n17 = NamedRerouteUsage()
  n18 = Constant() {r=3}
  n19 = Constant() {r=1}
  n20 = Add(A=NamedRerouteUsage())
  NamedRerouteDeclaration(None=SmoothStep(
    Min=NamedRerouteUsage()
    Value=MaterialFunctionCall(uv=@n17, blockiness=@n18, complexity=@n19, timemultiplier=@n20) {materialFunction=&MF_FbmForGlitch}.value
   ) {constMin=0.5}) {name=noisevalblue, nodeColor=<rgba 0 0.1751 1 1>}
  NamedRerouteDeclaration(None=TextureSample(UVs=NamedRerouteUsage(), Tex=NamedRerouteUsage()) {samplerSource=SSM_Wrap_WorldGroupSettings}.RGB) {name=blue_channel, nodeColor=<rgba 0.0599 0.112 1 1>}
  n21 = ComponentMask(None=NamedRerouteUsage()) {r=true}
  n22 = Saturate(None=NamedRerouteUsage())
  n23 = NamedRerouteUsage()
  NamedRerouteDeclaration(None=MaterialFunctionCall(X=Multiply(A=@n21, B=@n22), Y=ComponentMask(None=@n23) {g=true}) {materialFunction=&MakeFloat2}.Result) {name=alpha_UVs, nodeColor=<rgba 1 1 1 1>}
  NamedRerouteDeclaration(None=Constant() {r=0.2}) {name=glitchstepmin, nodeColor=<rgba 0 0.6719 1 1>}
  NamedRerouteDeclaration(None=TextureSample(UVs=NamedRerouteUsage(), Tex=NamedRerouteUsage()) {samplerSource=SSM_Wrap_WorldGroupSettings}.RGB) {name=alpha_channel, nodeColor=<rgba 1 1 1 1>}
  NamedRerouteDeclaration(None=Constant() {r=0.4}) {name=glitchstepmax, nodeColor=<rgba 0.9922 1 0 1>}
  n24 = FunctionInput() {bUsePreviewValueAsDefault=true, inputName=RedNoiseLowValue, inputType=FunctionInput_Scalar, sortPriority=2, previewValue=<0.5 0 0 1>}
  NamedRerouteDeclaration(None=@n24) {name=RedNoiseLowValue, nodeColor=<rgba 1 0.0087 0.0025 1>}
  n25 = TextureCoordinate()
  NamedRerouteDeclaration(None=NamedRerouteDeclaration(None=FunctionInput(Preview=@n25) {inputName=uv, inputType=FunctionInput_Vector2}) {name=uv, nodeColor=<rgba 0.5937 1 0 1>}) {name=uv2, nodeColor=<rgba 0 1 0.7656 1>}
  n26 = FunctionInput() {bUsePreviewValueAsDefault=true, inputName=GreenNoiseLowValue, inputType=FunctionInput_Scalar, sortPriority=3, previewValue=<0.5 0 0 1>}
  NamedRerouteDeclaration(None=@n26) {name=GreenNoiseLowValue, nodeColor=<rgba 0 1 0.1328 1>}
  n27 = FunctionInput() {bUsePreviewValueAsDefault=true, inputName=BlueNoiseLowValue, inputType=FunctionInput_Scalar, sortPriority=4, previewValue=<0.5 0 0 1>}
  NamedRerouteDeclaration(None=@n27) {name=BlueNoiseLowValue, nodeColor=<rgba 0.0603 0.2408 1 1>}
  n28 = FunctionInput() {inputName=glitchtimemultiplier, inputType=FunctionInput_Scalar, sortPriority=1, previewValue=<20 0 0 1>}
  NamedRerouteDeclaration(None=@n28) {name=GlitchTimeMultiplier, nodeColor=<rgba 0.1953 0 1 1>}
  n29 = NamedRerouteUsage()
  n30 = Constant() {r=0.5}
  n31 = Constant() {r=0.5}
  n32 = Add(A=NamedRerouteUsage()) {constB=3}
  NamedRerouteDeclaration(None=SmoothStep(
    Min=NamedRerouteUsage()
    Value=MaterialFunctionCall(uv=@n29, blockiness=@n30, complexity=@n31, timemultiplier=@n32) {materialFunction=&MF_FbmForGlitch}.value
   ) {constMin=0.5}) {name=noisevalred, nodeColor=<rgba 1 0.0588 0.0941 1>}
  n33 = ComponentMask(None=NamedRerouteUsage()) {r=true}
  n34 = MaterialFunctionCall(uv=NamedRerouteUsage(), blockiness=Constant() {r=3}, complexity=Constant() {r=1}, timemultiplier=NamedRerouteUsage()) {materialFunction=&MF_FbmForGlitch}
  n35 = NamedRerouteUsage()
  NamedRerouteDeclaration(None=MaterialFunctionCall(X=Multiply(A=@n33, B=@n34.value), Y=ComponentMask(None=@n35) {g=true}) {materialFunction=&MakeFloat2}.Result) {name=uv_update2, nodeColor=<rgba 1 0.4954 0.5026 1>}
  n36 = TextureObjectParameter() {
   parameterName=GlitchTexture
   samplerType=SAMPLERTYPE_LinearColor
   texture=/BRCosmetics/Characters/Player/Male/Medium/Bodies/M_MED_ByteGear/Skins/Roll/Textures/T_ByteGear_Roll_GlitchRGB_FX
  }
  NamedRerouteDeclaration(None=@n36) {name=TextureObject, nodeColor=<rgba 1 0 0.5156 1>}
  n37 = ComponentMask(None=NamedRerouteUsage()) {r=true}
  n38 = Multiply(A=NamedRerouteUsage()) {constB=0.05}
  n39 = NamedRerouteUsage()
  NamedRerouteDeclaration(None=MaterialFunctionCall(X=Add(A=@n37, B=@n38), Y=ComponentMask(None=@n39) {g=true}) {materialFunction=&MakeFloat2}.Result) {name=red_UVs, nodeColor=<rgba 1 0.1408 0.1202 1>}
  NamedRerouteDeclaration(None=TextureSample(UVs=NamedRerouteUsage(), Tex=NamedRerouteUsage()) {samplerSource=SSM_Wrap_WorldGroupSettings}.R) {name=red_channel, nodeColor=<rgba 1 0.0725 0.1016 1>}
  n40 = ComponentMask(None=NamedRerouteUsage()) {r=true}
  n41 = Multiply(A=NamedRerouteUsage()) {constB=0.05}
  n42 = NamedRerouteUsage()
  NamedRerouteDeclaration(None=MaterialFunctionCall(X=Subtract(A=@n40, B=@n41), Y=ComponentMask(None=@n42) {g=true}) {materialFunction=&MakeFloat2}.Result) {name=green_UVs, nodeColor=<rgba 0.1455 1 0.2166 1>}
  NamedRerouteUsage()
 ]}
