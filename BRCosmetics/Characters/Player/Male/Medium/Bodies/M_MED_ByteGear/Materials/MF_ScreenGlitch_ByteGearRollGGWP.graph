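# Material function for the ByteGear Roll character's GGWP screen glitch that takes texture UVs and outputs distorted UVs.
# A time-driven sine, smoothstepped and multiplied by random flicker, forms a glitch frequency mask.
# Stretched, horizontal, big and sine-wiggle distortions are built from panned UVs, noise patterns and nested glitch functions, with static switches toggling layers.
# Ceil-quantized time produces stepped animation offsets, while box masks and frequency masks confine localized glitch bursts.
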MaterialFunction /BRCosmetics/Characters/Player/Male/Medium/Bodies/M_MED_ByteGear/Materials/MF_ScreenGlitch_ByteGearRollGGWP
graph {expressions {
  MaterialExpressionAdd_0 {
   class=MaterialExpressionAdd
   inputs [
    A {from=MaterialExpressionStaticSwitchParameter_1}
    B {from=MaterialExpressionNamedRerouteUsage_8}
   ]
  }
  MaterialExpressionAdd_1 {
   class=MaterialExpressionAdd
   inputs [
    A {from=MaterialExpressionSubtract_11}
    B {from=MaterialExpressionConvert_0}
   ]
  }
  MaterialExpressionAdd_10 {
   class=MaterialExpressionAdd
   inputs [
    A {from=MaterialExpressionMultiply_4}
    B {from=MaterialExpressionNamedRerouteUsage_7}
   ]
  }
  MaterialExpressionAdd_11 {
   class=MaterialExpressionAdd
   inputs [
    A {from=MaterialExpressionAdd_22}
    B {from=MaterialExpressionConvert_1}
   ]
  }
  MaterialExpressionAdd_12 {
   class=MaterialExpressionAdd
   inputs [
    A {from=MaterialExpressionNamedRerouteUsage_32}
    B {from=MaterialExpressionConvert_2}
   ]
  }
  MaterialExpressionAdd_13 {
   class=MaterialExpressionAdd
   inputs [
    A {from=MaterialExpressionMaterialFunctionCall_6, output=RedChannel}
    B {from=MaterialExpressionMaterialFunctionCall_6, output=RedChannel}
   ]
  }
  MaterialExpressionAdd_14 {
   class=MaterialExpressionAdd
   inputs [
    A {from=MaterialExpressionAdd_13}
    B {from=MaterialExpressionMaterialFunctionCall_6, output=BlueChannel}
   ]
  }
  MaterialExpressionAdd_15 {
   class=MaterialExpressionAdd
   inputs [
    A {from=MaterialExpressionAdd_14}
    B {from=MaterialExpressionMaterialFunctionCall_6, output=AlphaChannel}
   ]
  }
  MaterialExpressionAdd_16 {
   class=MaterialExpressionAdd
   inputs [
    A {from=MaterialExpressionConstant2Vector_11}
    B {from=MaterialExpressionDivide_3}
   ]
  }
  MaterialExpressionAdd_17 {
   class=MaterialExpressionAdd
   inputs [
    A {from=MaterialExpressionConstant_16}
    B {from=MaterialExpressionMultiply_30}
   ]
  }
  MaterialExpressionAdd_18 {
   class=MaterialExpressionAdd
   inputs [
    A {from=MaterialExpressionConstant_15}
    B {from=MaterialExpressionMultiply_32}
   ]
  }
  MaterialExpressionAdd_19 {
   class=MaterialExpressionAdd
   inputs [
    A {from=""}
    B {from=MaterialExpressionStaticSwitchParameter_5}
   ]
  }
  MaterialExpressionAdd_21 {
   class=MaterialExpressionAdd
   inputs [
    A {from=MaterialExpressionMultiply_34}
    B {from=MaterialExpressionNamedRerouteUsage_18}
   ]
  }
  MaterialExpressionAdd_22 {
   class=MaterialExpressionAdd
   inputs [
    A {from=MaterialExpressionTextureCoordinate_2}
    B {from=MaterialExpressionConstant2Vector_1}
   ]
  }
  MaterialExpressionAdd_28 {
   class=MaterialExpressionAdd
   inputs [
    A {from=MaterialExpressionAdd_16}
    B {from=MaterialExpressionMultiply_39}
   ]
  }
  MaterialExpressionAdd_29 {
   class=MaterialExpressionAdd
   inputs [
    A {from=MaterialExpressionAppendVector_2}
    B {from=MaterialExpressionNamedRerouteUsage_20}
   ]
  }
  MaterialExpressionAdd_3 {
   class=MaterialExpressionAdd
   inputs [
    A {from=MaterialExpressionAppendVector_3}
    B {from=MaterialExpressionStaticSwitchParameter_5}
   ]
  }
  MaterialExpressionAdd_4 {
   class=MaterialExpressionAdd
   inputs [
    A {from=MaterialExpressionStaticSwitchParameter_4}
    B {from=MaterialExpressionNamedRerouteUsage_5}
   ]
  }
  MaterialExpressionAdd_5 {
   class=MaterialExpressionAdd
   inputs [
    A {from=MaterialExpressionMultiply_8}
    B {from=""}
   ]
   properties {constB=0.5}
  }
  MaterialExpressionAdd_6 {
   class=MaterialExpressionAdd
   inputs [
    A {from=MaterialExpressionMultiply_19}
    B {from=MaterialExpressionDivide_1}
   ]
  }
  MaterialExpressionAdd_7 {
   class=MaterialExpressionAdd
   inputs [
    A {from=MaterialExpressionStaticSwitchParameter_2}
    B {from=MaterialExpressionStaticSwitchParameter_0}
   ]
  }
  MaterialExpressionAdd_8 {
   class=MaterialExpressionAdd
   inputs [
    A {from=MaterialExpressionSubtract_8}
    B {from=""}
   ]
   properties {constB=0.3}
  }
  MaterialExpressionAdd_9 {
   class=MaterialExpressionAdd
   inputs [
    A {from=MaterialExpressionMultiply_20}
    B {from=MaterialExpressionLinearInterpolate_2}
   ]
  }
  MaterialExpressionAppendVector_0 {
   class=MaterialExpressionAppendVector
   inputs [
    A {from=MaterialExpressionConstant_1, output=""}
    B {from=MaterialExpressionScalarParameter_10, output=""}
   ]
  }
  MaterialExpressionAppendVector_1 {
   class=MaterialExpressionAppendVector
   inputs [
    A {from=MaterialExpressionSubtract_3, output=""}
    B {from=MaterialExpressionSubtract_4, output=""}
   ]
  }
  MaterialExpressionAppendVector_2 {
   class=MaterialExpressionAppendVector
   inputs [
    A {from=MaterialExpressionMultiply_2, output=""}
    B {from=MaterialExpressionMultiply_48, output=""}
   ]
  }
  MaterialExpressionAppendVector_3 {
   class=MaterialExpressionAppendVector
   inputs [
    A {from=MaterialExpressionAdd_1, output=""}
    B {from=MaterialExpressionConvert_0, output=""}
   ]
  }
  MaterialExpressionAppendVector_4 {
   class=MaterialExpressionAppendVector
   inputs [
    A {from=MaterialExpressionConstant_6, output=""}
    B {from=MaterialExpressionScalarParameter_13, output=""}
   ]
  }
  MaterialExpressionAppendVector_5 {
   class=MaterialExpressionAppendVector
   inputs [
    A {from=MaterialExpressionMultiply_21, output=""}
    B {from=MaterialExpressionConstant_7, output=""}
   ]
  }
  MaterialExpressionAppendVector_6 {
   class=MaterialExpressionAppendVector
   inputs [
    A {from=MaterialExpressionMultiply_11, output=""}
    B {from=MaterialExpressionMultiply_50, output=""}
   ]
  }
  MaterialExpressionCeil_0 {class=MaterialExpressionCeil, inputs [None {from=MaterialExpressionMultiply_29, output=""}], properties {}}
  MaterialExpressionCeil_1 {class=MaterialExpressionCeil, inputs [None {from=MaterialExpressionMultiply_37, output=""}], properties {}}
  MaterialExpressionComponentMask_0 {class=MaterialExpressionComponentMask, inputs [None {from=MaterialExpressionMultiply_3}], properties {b=true, r=true}}
  MaterialExpressionConstant2Vector_0 {class=MaterialExpressionConstant2Vector, properties {g=10}}
  MaterialExpressionConstant2Vector_1 {class=MaterialExpressionConstant2Vector, properties {g=0.367, r=0.1}}
  MaterialExpressionConstant2Vector_10 {class=MaterialExpressionConstant2Vector, properties {r=1}}
  MaterialExpressionConstant2Vector_11 {class=MaterialExpressionConstant2Vector, properties {g=0.091, r=0.505}}
  MaterialExpressionConstant2Vector_12 {class=MaterialExpressionConstant2Vector, properties {g=0.5, r=16}}
  MaterialExpressionConstant2Vector_13 {class=MaterialExpressionConstant2Vector, properties {g=0.1, r=0.1}}
  MaterialExpressionConstant2Vector_14 {class=MaterialExpressionConstant2Vector, properties {g=0.151, r=0.5}}
  MaterialExpressionConstant2Vector_15 {class=MaterialExpressionConstant2Vector, properties {g=0.05, r=0.08}}
  MaterialExpressionConstant2Vector_2 {class=MaterialExpressionConstant2Vector, properties {g=10}}
  MaterialExpressionConstant2Vector_3 {class=MaterialExpressionConstant2Vector, properties {g=0.5, r=32}}
  MaterialExpressionConstant2Vector_4 {class=MaterialExpressionConstant2Vector, properties {g=1, r=2}}
  MaterialExpressionConstant2Vector_6 {class=MaterialExpressionConstant2Vector, properties {g=0.2, r=0.1}}
  MaterialExpressionConstant2Vector_7 {class=MaterialExpressionConstant2Vector, properties {g=0.1, r=1}}
  MaterialExpressionConstant2Vector_8 {class=MaterialExpressionConstant2Vector, properties {g=7, r=1}}
  MaterialExpressionConstant2Vector_9 {class=MaterialExpressionConstant2Vector, properties {g=2, r=2}}
  MaterialExpressionConstant_0 {class=MaterialExpressionConstant}
  MaterialExpressionConstant_1 {class=MaterialExpressionConstant, properties {r=0.5}}
  MaterialExpressionConstant_10 {class=MaterialExpressionConstant, properties {r=1}}
  MaterialExpressionConstant_11 {class=MaterialExpressionConstant, properties {r=0.45}}
  MaterialExpressionConstant_12 {class=MaterialExpressionConstant, properties {r=0.45}}
  MaterialExpressionConstant_13 {class=MaterialExpressionConstant, properties {r=0.4}}
  MaterialExpressionConstant_14 {class=MaterialExpressionConstant, properties {desc=Wiggle Intensity, r=0.03}}
  MaterialExpressionConstant_15 {class=MaterialExpressionConstant, properties {r=0.2}}
  MaterialExpressionConstant_16 {class=MaterialExpressionConstant, properties {r=0.2}}
  MaterialExpressionConstant_17 {class=MaterialExpressionConstant, properties {r=0.25}}
  MaterialExpressionConstant_18 {class=MaterialExpressionConstant}
  MaterialExpressionConstant_19 {class=MaterialExpressionConstant, properties {r=0.25}}
  MaterialExpressionConstant_2 {class=MaterialExpressionConstant, properties {r=0.25}}
  MaterialExpressionConstant_20 {class=MaterialExpressionConstant, properties {r=1}}
  MaterialExpressionConstant_21 {class=MaterialExpressionConstant, properties {r=0.25}}
  MaterialExpressionConstant_22 {class=MaterialExpressionConstant, properties {r=0.25}}
  MaterialExpressionConstant_23 {class=MaterialExpressionConstant, properties {r=0.15}}
  MaterialExpressionConstant_24 {class=MaterialExpressionConstant, properties {r=0.04}}
  MaterialExpressionConstant_25 {class=MaterialExpressionConstant, properties {r=0.1}}
  MaterialExpressionConstant_26 {class=MaterialExpressionConstant, properties {r=1}}
  MaterialExpressionConstant_27 {class=MaterialExpressionConstant, properties {r=1}}
  MaterialExpressionConstant_28 {class=MaterialExpressionConstant, properties {r=0.8}}
  MaterialExpressionConstant_29 {class=MaterialExpressionConstant, properties {r=1}}
  MaterialExpressionConstant_3 {class=MaterialExpressionConstant, properties {r=0.4}}
  MaterialExpressionConstant_30 {class=MaterialExpressionConstant, properties {r=0.2}}
  MaterialExpressionConstant_31 {class=MaterialExpressionConstant, properties {r=0.1}}
  MaterialExpressionConstant_32 {class=MaterialExpressionConstant, properties {r=4}}
  MaterialExpressionConstant_33 {class=MaterialExpressionConstant, properties {r=4}}
  MaterialExpressionConstant_34 {class=MaterialExpressionConstant, properties {r=1}}
  MaterialExpressionConstant_35 {class=MaterialExpressionConstant, properties {r=1}}
  MaterialExpressionConstant_36 {class=MaterialExpressionConstant, properties {r=0.4}}
  MaterialExpressionConstant_37 {class=MaterialExpressionConstant, properties {r=1}}
  MaterialExpressionConstant_4 {class=MaterialExpressionConstant, properties {r=0.45}}
  MaterialExpressionConstant_5 {class=MaterialExpressionConstant, properties {r=35}}
  MaterialExpressionConstant_6 {class=MaterialExpressionConstant, properties {r=0.5}}
  MaterialExpressionConstant_7 {class=MaterialExpressionConstant}
  MaterialExpressionConstant_8 {class=MaterialExpressionConstant, properties {desc=Wiggle Intensity, r=0.02}}
  MaterialExpressionConstant_9 {class=MaterialExpressionConstant, properties {r=0.4}}
  MaterialExpressionConvert_0 {
   class=MaterialExpressionConvert
   inputs [None {from=MaterialExpressionTextureCoordinate_1}]
   properties {
    nodeName=Break Float2
    convertInputs [{
      type=Vector2
      expressionInput {
       expression=&self:MaterialExpressionTextureCoordinate_1
       inputName=None
       mask=0
       maskA=0
       maskB=0
       maskG=0
       maskR=0
       outputIndex=0
      }
     }]
    convertOutputs [{type=Scalar}, {type=Scalar}]
   }
  }
  MaterialExpressionConvert_1 {
   class=MaterialExpressionConvert
   inputs [
    None {from=MaterialExpressionMultiply_0}
    None {from=MaterialExpressionMultiply_9}
   ]
   properties {
    nodeName=Make Float2
    convertInputs [
     {
      type=Scalar
      expressionInput {
       expression=&self:MaterialExpressionMultiply_0
       inputName=None
       mask=0
       maskA=0
       maskB=0
       maskG=0
       maskR=0
       outputIndex=0
      }
     }
     {
      type=Scalar
      expressionInput {
       expression=&self:MaterialExpressionMultiply_9
       inputName=None
       mask=0
       maskA=0
       maskB=0
       maskG=0
       maskR=0
       outputIndex=0
      }
     }
    ]
    convertOutputs [{type=Vector2}]
   }
  }
  MaterialExpressionConvert_2 {
   class=MaterialExpressionConvert
   inputs [
    None {from=MaterialExpressionMultiply_7}
    None {from=MaterialExpressionMultiply_22}
   ]
   properties {
    nodeName=Make Float2
    convertInputs [
     {
      type=Scalar
      expressionInput {
       expression=&self:MaterialExpressionMultiply_7
       inputName=None
       mask=0
       maskA=0
       maskB=0
       maskG=0
       maskR=0
       outputIndex=0
      }
     }
     {
      type=Scalar
      expressionInput {
       expression=&self:MaterialExpressionMultiply_22
       inputName=None
       mask=0
       maskA=0
       maskB=0
       maskG=0
       maskR=0
       outputIndex=0
      }
     }
    ]
    convertOutputs [{type=Vector2}]
   }
  }
  MaterialExpressionDivide_0 {
   class=MaterialExpressionDivide
   inputs [
    A {from=MaterialExpressionCeil_0}
    B {from=""}
   ]
   properties {constB=16}
  }
  MaterialExpressionDivide_1 {
   class=MaterialExpressionDivide
   inputs [
    A {from=MaterialExpressionFloor_1}
    B {from=""}
   ]
   properties {constB=3}
  }
  MaterialExpressionDivide_2 {
   class=MaterialExpressionDivide
   inputs [
    A {from=MaterialExpressionCeil_1}
    B {from=""}
   ]
   properties {constB=9}
  }
  MaterialExpressionDivide_3 {
   class=MaterialExpressionDivide
   inputs [
    A {from=MaterialExpressionConstant2Vector_6}
    B {from=""}
   ]
  }
  MaterialExpressionFloor_1 {class=MaterialExpressionFloor, inputs [None {from=MaterialExpressionMultiply_18, output=""}], properties {}}
  MaterialExpressionFunctionInput_0 {
   class=MaterialExpressionFunctionInput
   inputs [Preview {from=MaterialExpressionTextureCoordinate_0, output=""}]
   properties {bUsePreviewValueAsDefault=true, inputName=Main_TextureUV, inputType=FunctionInput_Vector2, sortPriority=2}
  }
  MaterialExpressionFunctionOutput_0 {class=MaterialExpressionFunctionOutput, inputs [None {from=MaterialExpressionLinearInterpolate_0, output=""}]}
  MaterialExpressionLinearInterpolate_0 {
   class=MaterialExpressionLinearInterpolate
   inputs [
    A {from=MaterialExpressionNamedRerouteUsage_11}
    B {from=MaterialExpressionAdd_29}
    Alpha {from=MaterialExpressionMultiply_40}
   ]
  }
  MaterialExpressionLinearInterpolate_1 {
   class=MaterialExpressionLinearInterpolate
   inputs [
    A {from=""}
    B {from=MaterialExpressionScalarParameter_12}
    Alpha {from=MaterialExpressionNamedRerouteUsage_14}
   ]
   properties {constB=0.1}
  }
  MaterialExpressionLinearInterpolate_2 {
   class=MaterialExpressionLinearInterpolate
   inputs [
    A {from=""}
    B {from=MaterialExpressionAppendVector_5}
    Alpha {from=MaterialExpressionNamedRerouteUsage_15}
   ]
   properties {constB=0.1}
  }
  MaterialExpressionLinearInterpolate_4 {
   class=MaterialExpressionLinearInterpolate
   inputs [
    A {from=MaterialExpressionReroute_0}
    B {from=MaterialExpressionMultiply_47}
    Alpha {from=MaterialExpressionNamedRerouteUsage_4}
   ]
  }
  MaterialExpressionLinearInterpolate_6 {
   class=MaterialExpressionLinearInterpolate
   inputs [
    A {from=""}
    B {from=MaterialExpressionScalarParameter_9}
    Alpha {from=MaterialExpressionNamedRerouteUsage_2}
   ]
   properties {constB=0.1}
  }
  MaterialExpressionLocalPosition_0 {class=MaterialExpressionLocalPosition, inputs [], properties {includedOffsets=ExcludeOffsets, localOrigin=InstancePreSkinning}}
  MaterialExpressionMaterialFunctionCall_0 {
   class=MaterialExpressionMaterialFunctionCall
   inputs [
    UVs {from=MaterialExpressionTextureCoordinate_3}
    Rotation Center {from=""}
    "Rotation Angle (0-1)" {from=MaterialExpressionConstant_2}
   ]
   properties {materialFunction=&CustomRotator}
  }
  MaterialExpressionMaterialFunctionCall_10 {
   class=MaterialExpressionMaterialFunctionCall
   inputs [
    GlitchFrequency {from=MaterialExpressionConstant_20}
    GlitchFrequencyMultiplier {from=MaterialExpressionConstant_21}
    GlitchDuration {from=MaterialExpressionConstant_22}
    GlitchDurationMultiplier {from=MaterialExpressionConstant_23}
   ]
   properties {materialFunction=&/MF_GlitchFrequencyMask}
  }
  MaterialExpressionMaterialFunctionCall_11 {
   class=MaterialExpressionMaterialFunctionCall
   inputs [
    uv {from=MaterialExpressionMultiply_24}
    glitchtimemultiplier {from=MaterialExpressionConstant_5}
    RedNoiseLowValue {from=MaterialExpressionConstant_4}
    GreenNoiseLowValue {from=MaterialExpressionConstant_11}
    BlueNoiseLowValue {from=MaterialExpressionConstant_12}
   ]
   properties {materialFunction=/BRCosmetics/Characters/Player/Male/Medium/Bodies/M_MED_ByteGear/Skins/Roll/Materials/MF_GlitchPattern}
  }
  MaterialExpressionMaterialFunctionCall_12 {
   class=MaterialExpressionMaterialFunctionCall
   inputs [
    A {from=MaterialExpressionTextureCoordinate_5}
    B {from=MaterialExpressionAdd_28}
    Bounds {from=MaterialExpressionConstant2Vector_6}
    Edge Falloff {from=MaterialExpressionConstant_24}
   ]
   properties {materialFunction=/Engine/Functions/Engine_MaterialFunctions01/Gradient/BoxMask-2D}
  }
  MaterialExpressionMaterialFunctionCall_13 {
   class=MaterialExpressionMaterialFunctionCall
   inputs [
    A {from=MaterialExpressionTextureCoordinate_6}
    B {from=MaterialExpressionConstant2Vector_14}
    Bounds {from=MaterialExpressionConstant2Vector_13}
    Edge Falloff {from=MaterialExpressionConstant_25}
   ]
   properties {materialFunction=&BoxMask-2D}
  }
  MaterialExpressionMaterialFunctionCall_14 {
   class=MaterialExpressionMaterialFunctionCall
   inputs [
    GlitchFrequency {from=MaterialExpressionConstant_28}
    GlitchFrequencyMultiplier {from=MaterialExpressionConstant_29}
    GlitchDuration {from=MaterialExpressionConstant_26}
    GlitchDurationMultiplier {from=MaterialExpressionConstant_27}
   ]
   properties {materialFunction=&MF_GlitchFrequencyMask}
  }
  MaterialExpressionMaterialFunctionCall_15 {
   class=MaterialExpressionMaterialFunctionCall
   inputs [
    GlitchFrequency {from=MaterialExpressionConstant_36}
    GlitchFrequencyMultiplier {from=MaterialExpressionConstant_37}
    GlitchDuration {from=MaterialExpressionConstant_34}
    GlitchDurationMultiplier {from=MaterialExpressionConstant_35}
   ]
   properties {materialFunction=&MF_GlitchFrequencyMask}
  }
  MaterialExpressionMaterialFunctionCall_18 {
   class=MaterialExpressionMaterialFunctionCall
   inputs [Main_TextureUV {from=MaterialExpressionAdd_11}]
   properties {materialFunction=/Game/Global/Materials/MaterialGym/Materials/ScreenGlitchGeneric/MF_ScreenGlitch_Generic}
  }
  MaterialExpressionMaterialFunctionCall_19 {
   class=MaterialExpressionMaterialFunctionCall
   inputs [Main_TextureUV {from=MaterialExpressionAdd_12}]
   properties {materialFunction=&MF_ScreenGlitch_Generic}
  }
  MaterialExpressionMaterialFunctionCall_2 {class=MaterialExpressionMaterialFunctionCall, inputs [], properties {materialFunction=/Game/MF__FloatFlickerRandom_Generic}}
  MaterialExpressionMaterialFunctionCall_24 {
   class=MaterialExpressionMaterialFunctionCall
   inputs [
    UVs {from=MaterialExpressionMaterialFunctionCall_18, output=Result}
    Rotation Center {from=""}
    "Rotation Angle (0-1)" {from=MaterialExpressionConstant_18}
   ]
   properties {materialFunction=&CustomRotator}
  }
  MaterialExpressionMaterialFunctionCall_3 {class=MaterialExpressionMaterialFunctionCall, inputs [], properties {materialFunction=&MF__FloatFlickerRandom_Generic}}
  MaterialExpressionMaterialFunctionCall_4 {
   class=MaterialExpressionMaterialFunctionCall
   inputs [UV {from=MaterialExpressionNamedRerouteDeclaration_0}]
   properties {materialFunction=&/MF_HolographicTextureGlitchByteGearRoll_Grid}
  }
  MaterialExpressionMaterialFunctionCall_5 {
   class=MaterialExpressionMaterialFunctionCall
   inputs [
    GlitchFrequency {from=MaterialExpressionConstant_30}
    GlitchFrequencyMultiplier {from=MaterialExpressionConstant_31}
    GlitchDuration {from=MaterialExpressionConstant_32}
    GlitchDurationMultiplier {from=MaterialExpressionConstant_33}
   ]
   properties {materialFunction=&MF_GlitchFrequencyMask}
  }
  MaterialExpressionMaterialFunctionCall_6 {
   class=MaterialExpressionMaterialFunctionCall
   inputs [
    uv {from=MaterialExpressionMultiply_23}
    glitchtimemultiplier {from=MaterialExpressionScalarParameter_0}
    RedNoiseLowValue {from=MaterialExpressionConstant_3}
    GreenNoiseLowValue {from=MaterialExpressionConstant_9}
    BlueNoiseLowValue {from=MaterialExpressionConstant_13}
   ]
   properties {materialFunction=&MF_GlitchPattern}
  }
  MaterialExpressionMaterialFunctionCall_7 {class=MaterialExpressionMaterialFunctionCall, inputs [], properties {materialFunction=&MF__FloatFlickerRandom_Generic}}
  MaterialExpressionMaterialFunctionCall_8 {
   class=MaterialExpressionMaterialFunctionCall
   inputs [
    UVs {from=MaterialExpressionTextureCoordinate_4}
    Rotation Center {from=""}
    "Rotation Angle (0-1)" {from=MaterialExpressionConstant_19}
   ]
   properties {materialFunction=&CustomRotator}
  }
  MaterialExpressionMaterialFunctionCall_9 {
   class=MaterialExpressionMaterialFunctionCall
   inputs [
    UVs {from=MaterialExpressionMaterialFunctionCall_19, output=Result}
    Rotation Center {from=""}
    "Rotation Angle (0-1)" {from=MaterialExpressionConstant_17}
   ]
   properties {materialFunction=&CustomRotator}
  }
  MaterialExpressionMultiply_0 {
   class=MaterialExpressionMultiply
   inputs [
    A {from=MaterialExpressionTime_0}
    B {from=""}
   ]
   properties {constB=-0.021}
  }
  MaterialExpressionMultiply_1 {
   class=MaterialExpressionMultiply
   inputs [
    A {from=MaterialExpressionMaterialFunctionCall_3, output=RandomFlickerFloat}
    B {from=""}
   ]
   properties {constB=2}
  }
  MaterialExpressionMultiply_10 {
   class=MaterialExpressionMultiply
   inputs [
    A {from=MaterialExpressionTime_1}
    B {from=MaterialExpressionAdd_18}
   ]
   properties {constB=0.2}
  }
  MaterialExpressionMultiply_11 {
   class=MaterialExpressionMultiply
   inputs [
    A {from=MaterialExpressionSine_2}
    B {from=MaterialExpressionConstant_8}
   ]
   properties {constB=0.007}
  }
  MaterialExpressionMultiply_12 {
   class=MaterialExpressionMultiply
   inputs [
    A {from=MaterialExpressionTime_2}
    B {from=""}
   ]
   properties {constB=10}
  }
  MaterialExpressionMultiply_13 {
   class=MaterialExpressionMultiply
   inputs [
    A {from=MaterialExpressionNamedRerouteUsage_3}
    B {from=MaterialExpressionAppendVector_6}
   ]
  }
  MaterialExpressionMultiply_14 {
   class=MaterialExpressionMultiply
   inputs [
    A {from=MaterialExpressionSmoothStep_1}
    B {from=MaterialExpressionMaterialFunctionCall_7, output=RandomFlickerFloat}
   ]
  }
  MaterialExpressionMultiply_15 {
   class=MaterialExpressionMultiply
   inputs [
    A {from=MaterialExpressionRotator_1}
    B {from=MaterialExpressionConstant2Vector_2}
   ]
  }
  MaterialExpressionMultiply_16 {
   class=MaterialExpressionMultiply
   inputs [
    A {from=MaterialExpressionMultiply_17}
    B {from=MaterialExpressionReroute_2}
   ]
   properties {constB=0.2}
  }
  MaterialExpressionMultiply_17 {
   class=MaterialExpressionMultiply
   inputs [
    A {from=MaterialExpressionSubtract_7}
    B {from=MaterialExpressionSaturate_0}
   ]
  }
  MaterialExpressionMultiply_18 {
   class=MaterialExpressionMultiply
   inputs [
    A {from=MaterialExpressionTime_4}
    B {from=""}
   ]
   properties {constB=3}
  }
  MaterialExpressionMultiply_19 {
   class=MaterialExpressionMultiply
   inputs [
    A {from=MaterialExpressionNamedRerouteUsage_17}
    B {from=MaterialExpressionConstant2Vector_8}
   ]
   properties {constB=5}
  }
  MaterialExpressionMultiply_2 {
   class=MaterialExpressionMultiply
   inputs [
    A {from=""}
    B {from=MaterialExpressionSubtract_5}
   ]
   properties {constA=0.005, constB=0.1}
  }
  MaterialExpressionMultiply_20 {
   class=MaterialExpressionMultiply
   inputs [
    A {from=MaterialExpressionConstant2Vector_10}
    B {from=MaterialExpressionMultiply_16}
   ]
  }
  MaterialExpressionMultiply_21 {
   class=MaterialExpressionMultiply
   inputs [
    A {from=""}
    B {from=MaterialExpressionScalarParameter_14}
   ]
   properties {constA=-0.75}
  }
  MaterialExpressionMultiply_22 {
   class=MaterialExpressionMultiply
   inputs [
    A {from=MaterialExpressionTime_5}
    B {from=""}
   ]
   properties {constB=0.013}
  }
  MaterialExpressionMultiply_23 {
   class=MaterialExpressionMultiply
   inputs [
    A {from=MaterialExpressionMaterialFunctionCall_9, output=Rotated Values}
    B {from=MaterialExpressionConstant2Vector_4}
   ]
  }
  MaterialExpressionMultiply_24 {
   class=MaterialExpressionMultiply
   inputs [
    A {from=MaterialExpressionMaterialFunctionCall_24, output=Rotated Values}
    B {from=MaterialExpressionConstant2Vector_9}
   ]
  }
  MaterialExpressionMultiply_25 {
   class=MaterialExpressionMultiply
   inputs [
    A {from=MaterialExpressionMaterialFunctionCall_2, output=RandomFlickerFloat}
    B {from=""}
   ]
   properties {constB=2}
  }
  MaterialExpressionMultiply_26 {
   class=MaterialExpressionMultiply
   inputs [
    A {from=MaterialExpressionTextureSample_2, output=R}
    B {from=""}
   ]
   properties {constB=2}
  }
  MaterialExpressionMultiply_27 {
   class=MaterialExpressionMultiply
   inputs [
    A {from=MaterialExpressionTextureSample_4, output=G}
    B {from=""}
   ]
   properties {constB=2}
  }
  MaterialExpressionMultiply_28 {
   class=MaterialExpressionMultiply
   inputs [
    A {from=MaterialExpressionConstant2Vector_7}
    B {from=MaterialExpressionPanner_2}
   ]
  }
  MaterialExpressionMultiply_29 {
   class=MaterialExpressionMultiply
   inputs [
    A {from=MaterialExpressionTime_3}
    B {from=MaterialExpressionConstant2Vector_3}
   ]
  }
  MaterialExpressionMultiply_3 {
   class=MaterialExpressionMultiply
   inputs [
    A {from=MaterialExpressionVertexInterpolator_0, output=PS}
    B {from=MaterialExpressionScalarParameter_7}
   ]
  }
  MaterialExpressionMultiply_30 {
   class=MaterialExpressionMultiply
   inputs [
    A {from=MaterialExpressionSubtract_6}
    B {from=""}
   ]
   properties {constB=0.1}
  }
  MaterialExpressionMultiply_31 {
   class=MaterialExpressionMultiply
   inputs [
    A {from=MaterialExpressionTextureSample_1, output=R}
    B {from=""}
   ]
   properties {constB=2}
  }
  MaterialExpressionMultiply_32 {
   class=MaterialExpressionMultiply
   inputs [
    A {from=MaterialExpressionSubtract_2}
    B {from=""}
   ]
   properties {constB=0.075}
  }
  MaterialExpressionMultiply_33 {
   class=MaterialExpressionMultiply
   inputs [
    A {from=""}
    B {from=MaterialExpressionTextureSample_3, output=R}
   ]
   properties {constA=0.005, constB=0.1}
  }
  MaterialExpressionMultiply_34 {
   class=MaterialExpressionMultiply
   inputs [
    A {from=MaterialExpressionMaterialFunctionCall_8, output=Rotated Values}
    B {from=""}
   ]
   properties {constB=12}
  }
  MaterialExpressionMultiply_35 {
   class=MaterialExpressionMultiply
   inputs [
    A {from=MaterialExpressionMultiply_33}
    B {from=MaterialExpressionMaterialFunctionCall_10, output=Result}
   ]
  }
  MaterialExpressionMultiply_36 {
   class=MaterialExpressionMultiply
   inputs [
    A {from=""}
    B {from=MaterialExpressionMaterialFunctionCall_12, output=Result}
   ]
   properties {constA=0.67}
  }
  MaterialExpressionMultiply_37 {
   class=MaterialExpressionMultiply
   inputs [
    A {from=MaterialExpressionTime_7}
    B {from=MaterialExpressionConstant2Vector_12}
   ]
  }
  MaterialExpressionMultiply_38 {
   class=MaterialExpressionMultiply
   inputs [
    A {from=MaterialExpressionMultiply_35}
    B {from=MaterialExpressionMaterialFunctionCall_13, output=Result}
   ]
  }
  MaterialExpressionMultiply_39 {
   class=MaterialExpressionMultiply
   inputs [
    A {from=MaterialExpressionAppendVector_1}
    B {from=MaterialExpressionConstant2Vector_15}
   ]
   properties {constB=0.15}
  }
  MaterialExpressionMultiply_4 {
   class=MaterialExpressionMultiply
   inputs [
    A {from=MaterialExpressionMaterialFunctionCall_0, output=Rotated Values}
    B {from=""}
   ]
   properties {constB=15}
  }
  MaterialExpressionMultiply_40 {
   class=MaterialExpressionMultiply
   inputs [
    A {from=MaterialExpressionMaterialFunctionCall_5, output=Result}
    B {from=MaterialExpressionMultiply_36}
   ]
  }
  MaterialExpressionMultiply_41 {
   class=MaterialExpressionMultiply
   inputs [
    A {from=MaterialExpressionMaterialFunctionCall_14, output=Result}
    B {from=""}
   ]
   properties {constB=2}
  }
  MaterialExpressionMultiply_42 {
   class=MaterialExpressionMultiply
   inputs [
    A {from=MaterialExpressionMaterialFunctionCall_15, output=Result}
    B {from=""}
   ]
   properties {constB=2}
  }
  MaterialExpressionMultiply_43 {
   class=MaterialExpressionMultiply
   inputs [
    A {from=MaterialExpressionTextureSample_0, output=R}
    B {from=""}
   ]
   properties {constB=2}
  }
  MaterialExpressionMultiply_44 {
   class=MaterialExpressionMultiply
   inputs [
    A {from=MaterialExpressionSine_2}
    B {from=""}
   ]
   properties {constB=2}
  }
  MaterialExpressionMultiply_45 {
   class=MaterialExpressionMultiply
   inputs [
    A {from=MaterialExpressionMaterialFunctionCall_11, output=RedChannel}
    B {from=""}
   ]
   properties {constB=2}
  }
  MaterialExpressionMultiply_46 {
   class=MaterialExpressionMultiply
   inputs [
    A {from=MaterialExpressionMaterialFunctionCall_4, output=Glitched UV}
    B {from=""}
   ]
   properties {constB=2}
  }
  MaterialExpressionMultiply_47 {
   class=MaterialExpressionMultiply
   inputs [
    A {from=MaterialExpressionSubtract_12}
    B {from=""}
   ]
   properties {constB=0.005}
  }
  MaterialExpressionMultiply_48 {
   class=MaterialExpressionMultiply
   inputs [
    A {from=""}
    B {from=MaterialExpressionSubtract_13}
   ]
   properties {constA=0.005, constB=0.1}
  }
  MaterialExpressionMultiply_49 {
   class=MaterialExpressionMultiply
   inputs [
    A {from=MaterialExpressionTextureSample_0, output=G}
    B {from=""}
   ]
   properties {constB=2}
  }
  MaterialExpressionMultiply_5 {
   class=MaterialExpressionMultiply
   inputs [
    A {from=MaterialExpressionRotator_0}
    B {from=MaterialExpressionConstant2Vector_0}
   ]
  }
  MaterialExpressionMultiply_50 {
   class=MaterialExpressionMultiply
   inputs [
    A {from=MaterialExpressionSine_0}
    B {from=MaterialExpressionConstant_14}
   ]
   properties {constB=0.007}
  }
  MaterialExpressionMultiply_51 {
   class=MaterialExpressionMultiply
   inputs [
    A {from=MaterialExpressionTime_6}
    B {from=""}
   ]
   properties {constB=11}
  }
  MaterialExpressionMultiply_52 {
   class=MaterialExpressionMultiply
   inputs [
    A {from=MaterialExpressionSine_0}
    B {from=""}
   ]
   properties {constB=2}
  }
  MaterialExpressionMultiply_6 {
   class=MaterialExpressionMultiply
   inputs [
    A {from=MaterialExpressionSubtract_9}
    B {from=MaterialExpressionLinearInterpolate_6}
   ]
   properties {constB=0.2}
  }
  MaterialExpressionMultiply_7 {
   class=MaterialExpressionMultiply
   inputs [
    A {from=MaterialExpressionTime_5}
    B {from=""}
   ]
   properties {constB=0.027}
  }
  MaterialExpressionMultiply_8 {
   class=MaterialExpressionMultiply
   inputs [
    A {from=MaterialExpressionSine_1}
    B {from=""}
   ]
   properties {constB=0.5}
  }
  MaterialExpressionMultiply_9 {
   class=MaterialExpressionMultiply
   inputs [
    A {from=MaterialExpressionTime_0}
    B {from=""}
   ]
   properties {constB=0.041}
  }
  MaterialExpressionNamedRerouteDeclaration_0 {
   class=MaterialExpressionNamedRerouteDeclaration
   inputs [None {from=MaterialExpressionFunctionInput_0}]
   properties {name=MainTexture_UV, nodeColor=<rgba 0.3828 0 1 1>}
  }
  MaterialExpressionNamedRerouteDeclaration_1 {
   class=MaterialExpressionNamedRerouteDeclaration
   inputs [None {from=MaterialExpressionMultiply_14}]
   properties {name=GlitchFrequencyMask, nodeColor=<rgba 0.9377 1 0.8953 1>}
  }
  MaterialExpressionNamedRerouteDeclaration_10 {
   class=MaterialExpressionNamedRerouteDeclaration
   inputs [None {from=MaterialExpressionComponentMask_0}]
   properties {name=PlanarUV, nodeColor=<rgba 0.6875 0 1 1>}
  }
  MaterialExpressionNamedRerouteDeclaration_13 {
   class=MaterialExpressionNamedRerouteDeclaration
   inputs [None {from=MaterialExpressionDivide_0}]
   properties {name=GlitchAnimOffsetA, nodeColor=<rgba 1 0 0.1875 1>}
  }
  MaterialExpressionNamedRerouteDeclaration_2 {
   class=MaterialExpressionNamedRerouteDeclaration
   inputs [None {from=MaterialExpressionMultiply_6}]
   properties {name=StretchedGlitch, nodeColor=<rgba 0.7109 0 1 1>}
  }
  MaterialExpressionNamedRerouteDeclaration_3 {
   class=MaterialExpressionNamedRerouteDeclaration
   inputs [None {from=MaterialExpressionAdd_9}]
   properties {name=HoriztonalGlitch, nodeColor=<rgba 1 0.2812 0 1>}
  }
  MaterialExpressionNamedRerouteDeclaration_6 {
   class=MaterialExpressionNamedRerouteDeclaration
   inputs [None {from=MaterialExpressionDivide_2}]
   properties {name=BoxMaskAnimOffset, nodeColor=<rgba 0.0782 1 0.7681 1>}
  }
  MaterialExpressionNamedRerouteDeclaration_7 {
   class=MaterialExpressionNamedRerouteDeclaration
   inputs [None {from=MaterialExpressionStaticSwitchParameter_5}]
   properties {name=BigGlitch, nodeColor=<rgba 0.4062 1 0 1>}
  }
  MaterialExpressionNamedRerouteUsage_0 {class=MaterialExpressionNamedRerouteUsage}
  MaterialExpressionNamedRerouteUsage_1 {class=MaterialExpressionNamedRerouteUsage}
  MaterialExpressionNamedRerouteUsage_11 {class=MaterialExpressionNamedRerouteUsage}
  MaterialExpressionNamedRerouteUsage_14 {class=MaterialExpressionNamedRerouteUsage}
  MaterialExpressionNamedRerouteUsage_15 {class=MaterialExpressionNamedRerouteUsage}
  MaterialExpressionNamedRerouteUsage_16 {class=MaterialExpressionNamedRerouteUsage}
  MaterialExpressionNamedRerouteUsage_17 {class=MaterialExpressionNamedRerouteUsage}
  MaterialExpressionNamedRerouteUsage_18 {class=MaterialExpressionNamedRerouteUsage}
  MaterialExpressionNamedRerouteUsage_2 {class=MaterialExpressionNamedRerouteUsage}
  MaterialExpressionNamedRerouteUsage_20 {class=MaterialExpressionNamedRerouteUsage}
  MaterialExpressionNamedRerouteUsage_3 {class=MaterialExpressionNamedRerouteUsage}
  MaterialExpressionNamedRerouteUsage_32 {class=MaterialExpressionNamedRerouteUsage}
  MaterialExpressionNamedRerouteUsage_4 {class=MaterialExpressionNamedRerouteUsage}
  MaterialExpressionNamedRerouteUsage_5 {class=MaterialExpressionNamedRerouteUsage}
  MaterialExpressionNamedRerouteUsage_7 {class=MaterialExpressionNamedRerouteUsage}
  MaterialExpressionNamedRerouteUsage_8 {class=MaterialExpressionNamedRerouteUsage}
  MaterialExpressionNamedRerouteUsage_9 {class=MaterialExpressionNamedRerouteUsage}
  MaterialExpressionPanner_0 {
   class=MaterialExpressionPanner
   inputs [
    Coordinate {from=MaterialExpressionMultiply_5}
    Time {from=""}
    Speed {from=MaterialExpressionAppendVector_0}
   ]
   properties {speedX=0.5, speedY=5}
  }
  MaterialExpressionPanner_2 {
   class=MaterialExpressionPanner
   inputs [
    Coordinate {from=MaterialExpressionMultiply_15}
    Time {from=""}
    Speed {from=MaterialExpressionAppendVector_4}
   ]
   properties {speedX=0.5, speedY=5}
  }
  MaterialExpressionReroute_0 {class=MaterialExpressionReroute, inputs [None {from=MaterialExpressionNamedRerouteDeclaration_0}]}
  MaterialExpressionReroute_2 {class=MaterialExpressionReroute, inputs [None {from=MaterialExpressionLinearInterpolate_1}]}
  MaterialExpressionRotator_0 {
   class=MaterialExpressionRotator
   inputs [
    Coordinate {from=MaterialExpressionNamedRerouteUsage_9, output=""}
    Time {from=MaterialExpressionScalarParameter_6, output=""}
   ]
   properties {}
  }
  MaterialExpressionRotator_1 {
   class=MaterialExpressionRotator
   inputs [
    Coordinate {from=MaterialExpressionNamedRerouteUsage_16, output=""}
    Time {from=MaterialExpressionScalarParameter_8, output=""}
   ]
   properties {}
  }
  MaterialExpressionSaturate_0 {class=MaterialExpressionSaturate, inputs [None {from=MaterialExpressionAdd_8}]}
  MaterialExpressionScalarParameter_0 {class=MaterialExpressionScalarParameter, properties {defaultValue=21, parameterName="glitchtimemultiplier (S)"}}
  MaterialExpressionScalarParameter_10 {class=MaterialExpressionScalarParameter, properties {defaultValue=5, group=GGWPTexture_Glitch, parameterName=GGWPStretchedScrollingSpeed}}
  MaterialExpressionScalarParameter_12 {
   class=MaterialExpressionScalarParameter
   properties {defaultValue=0.1, group=HolographicTexture_Glitch, parameterName=StretchedGlitchIntensity}
  }
  MaterialExpressionScalarParameter_13 {class=MaterialExpressionScalarParameter, properties {defaultValue=5, group=HolographicTexture_Glitch, parameterName=StretchedScrollingSpeed}}
  MaterialExpressionScalarParameter_14 {
   class=MaterialExpressionScalarParameter
   properties {defaultValue=0.1, group=HolographicTexture_Glitch, parameterName=StretchedGlitchIntensity}
  }
  MaterialExpressionScalarParameter_4 {
   class=MaterialExpressionScalarParameter
   properties {defaultValue=0.2, group=GGWPTexture_Glitch, parameterName=GGWPGlitchFrequency, sortPriority=10}
  }
  MaterialExpressionScalarParameter_5 {
   class=MaterialExpressionScalarParameter
   properties {defaultValue=0.075, group=GGWPTexture_Glitch, parameterName=GGWPGlitchDuration, sortPriority=10}
  }
  MaterialExpressionScalarParameter_6 {class=MaterialExpressionScalarParameter, properties {group=GGWPTexture_Glitch, parameterName=GGWPGlitchUVRotation}}
  MaterialExpressionScalarParameter_7 {class=MaterialExpressionScalarParameter, properties {defaultValue=0.0333, parameterName=PlanarUVScale}}
  MaterialExpressionScalarParameter_8 {class=MaterialExpressionScalarParameter, properties {group=HolographicTexture_Glitch, parameterName=GlitchUVRotation}}
  MaterialExpressionScalarParameter_9 {class=MaterialExpressionScalarParameter, properties {defaultValue=0.1, group=GGWPTexture_Glitch, parameterName=GGWPStretchedGlitchIntensity}}
  MaterialExpressionSine_0 {class=MaterialExpressionSine, inputs [None {from=MaterialExpressionMultiply_51}]}
  MaterialExpressionSine_1 {class=MaterialExpressionSine, inputs [None {from=MaterialExpressionMultiply_10}]}
  MaterialExpressionSine_2 {class=MaterialExpressionSine, inputs [None {from=MaterialExpressionMultiply_12}]}
  MaterialExpressionSmoothStep_1 {
   class=MaterialExpressionSmoothStep
   inputs [
    Min {from=MaterialExpressionSubtract_0, output=""}
    Max {from=MaterialExpressionConstant_10, output=""}
    Value {from=MaterialExpressionAdd_5, output=""}
   ]
   properties {constMax=0.9, constMin=0.8}
  }
  MaterialExpressionSmoothStep_2 {
   class=MaterialExpressionSmoothStep
   inputs [
    Min {from="", output=""}
    Max {from="", output=""}
    Value {from=MaterialExpressionTextureSample_2, output=R}
   ]
   properties {constMax=0.1}
  }
  MaterialExpressionStaticSwitchParameter_0 {
   class=MaterialExpressionStaticSwitchParameter
   inputs [
    True {from=MaterialExpressionMultiply_13, output=""}
    False {from=MaterialExpressionConstant_0, output=""}
   ]
   properties {defaultValue=true, group=GGWPTexture_Glitch, parameterName=GGWPUseWiggleGlitch}
  }
  MaterialExpressionStaticSwitchParameter_1 {
   class=MaterialExpressionStaticSwitchParameter
   inputs [
    True {from=MaterialExpressionAdd_4, output=""}
    False {from=MaterialExpressionStaticSwitchParameter_4, output=""}
   ]
   properties {defaultValue=true, group=GGWPTexture_Glitch, parameterName=GGWPUseStretchedGlitch}
  }
  MaterialExpressionStaticSwitchParameter_2 {
   class=MaterialExpressionStaticSwitchParameter
   inputs [
    True {from=MaterialExpressionAdd_0, output=""}
    False {from=MaterialExpressionStaticSwitchParameter_1, output=""}
   ]
   properties {defaultValue=true, group=GGWPTexture_Glitch, parameterName=GGWPUseHoriztonalGlitch}
  }
  MaterialExpressionStaticSwitchParameter_4 {
   class=MaterialExpressionStaticSwitchParameter
   inputs [
    True {from=MaterialExpressionLinearInterpolate_4, output=""}
    False {from=MaterialExpressionNamedRerouteUsage_1, output=""}
   ]
   properties {defaultValue=true, group=GGWPTexture_Glitch, parameterName=GGWPUseGridGlitch, sortPriority=0}
  }
  MaterialExpressionStaticSwitchParameter_5 {
   class=MaterialExpressionStaticSwitchParameter
   inputs [
    True {from=MaterialExpressionAdd_7, output=""}
    False {from=MaterialExpressionNamedRerouteUsage_0, output=""}
   ]
   properties {defaultValue=true, group=GGWPTexture_Glitch, parameterName=UseGGWPGlitch, sortPriority=-1}
  }
  MaterialExpressionSubtract_0 {
   class=MaterialExpressionSubtract
   inputs [
    A {from=MaterialExpressionConstant_10}
    B {from=MaterialExpressionAdd_17}
   ]
  }
  MaterialExpressionSubtract_10 {
   class=MaterialExpressionSubtract
   inputs [
    A {from=MaterialExpressionMultiply_44}
    B {from=""}
   ]
  }
  MaterialExpressionSubtract_11 {
   class=MaterialExpressionSubtract
   inputs [
    A {from=MaterialExpressionMultiply_45}
    B {from=""}
   ]
  }
  MaterialExpressionSubtract_12 {
   class=MaterialExpressionSubtract
   inputs [
    A {from=MaterialExpressionMultiply_46}
    B {from=""}
   ]
  }
  MaterialExpressionSubtract_13 {
   class=MaterialExpressionSubtract
   inputs [
    A {from=MaterialExpressionMultiply_49}
    B {from=""}
   ]
  }
  MaterialExpressionSubtract_14 {
   class=MaterialExpressionSubtract
   inputs [
    A {from=MaterialExpressionMultiply_52}
    B {from=""}
   ]
  }
  MaterialExpressionSubtract_2 {
   class=MaterialExpressionSubtract
   inputs [
    A {from=MaterialExpressionMultiply_1}
    B {from=""}
   ]
  }
  MaterialExpressionSubtract_3 {
   class=MaterialExpressionSubtract
   inputs [
    A {from=MaterialExpressionMultiply_41}
    B {from=""}
   ]
  }
  MaterialExpressionSubtract_4 {
   class=MaterialExpressionSubtract
   inputs [
    A {from=MaterialExpressionMultiply_42}
    B {from=""}
   ]
  }
  MaterialExpressionSubtract_5 {
   class=MaterialExpressionSubtract
   inputs [
    A {from=MaterialExpressionMultiply_43}
    B {from=""}
   ]
  }
  MaterialExpressionSubtract_6 {
   class=MaterialExpressionSubtract
   inputs [
    A {from=MaterialExpressionMultiply_25}
    B {from=""}
   ]
  }
  MaterialExpressionSubtract_7 {
   class=MaterialExpressionSubtract
   inputs [
    A {from=MaterialExpressionMultiply_26}
    B {from=""}
   ]
  }
  MaterialExpressionSubtract_8 {
   class=MaterialExpressionSubtract
   inputs [
    A {from=MaterialExpressionMultiply_27}
    B {from=""}
   ]
  }
  MaterialExpressionSubtract_9 {
   class=MaterialExpressionSubtract
   inputs [
    A {from=MaterialExpressionMultiply_31}
    B {from=""}
   ]
  }
  MaterialExpressionTextureCoordinate_0 {class=MaterialExpressionTextureCoordinate}
  MaterialExpressionTextureCoordinate_1 {class=MaterialExpressionTextureCoordinate}
  MaterialExpressionTextureCoordinate_2 {class=MaterialExpressionTextureCoordinate}
  MaterialExpressionTextureCoordinate_3 {class=MaterialExpressionTextureCoordinate}
  MaterialExpressionTextureCoordinate_4 {class=MaterialExpressionTextureCoordinate}
  MaterialExpressionTextureCoordinate_5 {class=MaterialExpressionTextureCoordinate}
  MaterialExpressionTextureCoordinate_6 {class=MaterialExpressionTextureCoordinate}
  MaterialExpressionTextureObjectParameter_0 {
   class=MaterialExpressionTextureObjectParameter
   inputs [
    UVs {from="", output=""}
    Apply View MipBias {from="", output=""}
   ]
   properties {
    parameterName=GlitchUVTexture
    samplerType=SAMPLERTYPE_LinearColor
    texture=/BRCosmetics/Characters/Player/Male/Medium/Bodies/M_MED_ByteGear/Skins/Roll/Textures/T_ByteGear_Roll_Glitch_FX
   }
  }
  MaterialExpressionTextureObjectParameter_1 {
   class=MaterialExpressionTextureObjectParameter
   inputs [
    UVs {from="", output=""}
    Apply View MipBias {from="", output=""}
   ]
   properties {
    parameterName=GlitchUVTexture
    samplerType=SAMPLERTYPE_LinearColor
    texture=/BRCosmetics/Weapons/WeaponSkins/Wraps/Textures/ByteGear/T_ByteGear_Roll_GlitchRGB_FX
   }
  }
  MaterialExpressionTextureSample_0 {
   class=MaterialExpressionTextureSample
   inputs [
    UVs {from=MaterialExpressionAdd_10}
    Tex {from=MaterialExpressionTextureObjectParameter_1}
    Apply View MipBias {from=""}
   ]
   properties {
    samplerSource=SSM_Wrap_WorldGroupSettings
    samplerType=SAMPLERTYPE_LinearColor
    texture=&T_ByteGear_Roll_Glitch_FX
   }
  }
  MaterialExpressionTextureSample_1 {
   class=MaterialExpressionTextureSample
   inputs [
    UVs {from=MaterialExpressionPanner_0}
    Tex {from=""}
    Apply View MipBias {from=""}
   ]
   properties {samplerType=SAMPLERTYPE_Masks, texture=/Game/Characters/Player/Male/Medium/Bodies/M_MED_Storm_Soldier/Textures/ElectricityNoise_M}
  }
  MaterialExpressionTextureSample_2 {
   class=MaterialExpressionTextureSample
   inputs [
    UVs {from=MaterialExpressionMultiply_28}
    Tex {from=""}
    Apply View MipBias {from=""}
   ]
   properties {samplerType=SAMPLERTYPE_LinearColor, texture=/Game/Environments/Asteria/Sets/NeonCity/Techart/Holograms/GlitchMap_M}
  }
  MaterialExpressionTextureSample_3 {
   class=MaterialExpressionTextureSample
   inputs [
    UVs {from=MaterialExpressionAdd_21}
    Tex {from=MaterialExpressionTextureObjectParameter_0}
    Apply View MipBias {from=""}
   ]
   properties {
    samplerSource=SSM_Wrap_WorldGroupSettings
    samplerType=SAMPLERTYPE_LinearColor
    texture=&T_ByteGear_Roll_Glitch_FX
   }
  }
  MaterialExpressionTextureSample_4 {
   class=MaterialExpressionTextureSample
   inputs [
    UVs {from=MaterialExpressionAdd_6}
    Tex {from=""}
    Apply View MipBias {from=""}
   ]
   properties {samplerType=SAMPLERTYPE_LinearColor, texture=&GlitchMap_M}
  }
  MaterialExpressionTime_0 {class=MaterialExpressionTime}
  MaterialExpressionTime_1 {class=MaterialExpressionTime}
  MaterialExpressionTime_2 {class=MaterialExpressionTime}
  MaterialExpressionTime_3 {class=MaterialExpressionTime}
  MaterialExpressionTime_4 {class=MaterialExpressionTime}
  MaterialExpressionTime_5 {class=MaterialExpressionTime}
  MaterialExpressionTime_6 {class=MaterialExpressionTime}
  MaterialExpressionTime_7 {class=MaterialExpressionTime}
  MaterialExpressionVertexInterpolator_0 {class=MaterialExpressionVertexInterpolator, inputs [VS {from=MaterialExpressionLocalPosition_0, output=XYZ}], properties {}}
 }}
