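# Material function for the ByteGear Roll skin that layers a tinted holographic texture with scanlines and screen glitch onto incoming material attributes.
# A scanline subfunction modulates hologram color and opacity, blended by a switchable texture-channel, separate or full mask.
# A glitch subfunction supplies distorted UVs, while a time-driven sine threshold multiplied by random flicker forms the glitch frequency mask.
# It outputs modified attributes plus emissive, opacity, glitch UVs and that mask.
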MaterialFunction /BRCosmetics/Characters/Player/Male/Medium/Bodies/M_MED_ByteGear/Skins/Roll/Materials/MF_ByteGearRoll_HolographicTexture
graph {nodes [
  FunctionOutput(None=SetMaterialAttributes!(MaterialAttributes=NamedRerouteUsage(), EmissiveColor=NamedRerouteUsage(), Opacity=NamedRerouteUsage(), BaseColor=NamedRerouteUsage()) {attributeSetTypes [#1, #2, #3]}) {outputName=MaterialAttributes, sortPriority=-1}
  FunctionOutput(None=NamedRerouteUsage()) {outputName=EmissiveColor, sortPriority=1}
  FunctionOutput(None=NamedRerouteUsage()) {outputName=Opacity}
  FunctionOutput(None=NamedRerouteUsage()) {outputName=HolographicGlitchUVs, sortPriority=2}
  FunctionOutput(None=NamedRerouteUsage()) {outputName=GlitchFrequencyMask, sortPriority=3}
  NamedRerouteDeclaration(None=NamedRerouteUsage()) {name=BaseColor, nodeColor=<rgba 0.9453 1 0 1>}
  n1 = Constant() {r=1}
  n2 = Subtract(A=@n1, B=ScalarParameter() {defaultValue=0.075, group=HolographicTexture_Glitch, parameterName=GlitchDuration, sortPriority=10})
  n3 = Multiply(A=Time(), B=ScalarParameter() {defaultValue=0.2, group=HolographicTexture_Glitch, parameterName=GlitchFrequency, sortPriority=10}) {constB=0.2}
  n4 = Add(A=Multiply(A=Sine(None=@n3)) {constB=0.5}) {constB=0.5}
  NamedRerouteDeclaration(None=Multiply(
    A=SmoothStep(Min=@n2, Max=@n1, Value=@n4) {constMax=0.9, constMin=0.8}
    B=MaterialFunctionCall() {materialFunction=/Game/MF__FloatFlickerRandom_Generic}.RandomFlickerFloat
   )) {name=GlitchFrequencyMask, nodeColor=<rgba 0.9377 1 0.8953 1>}
  n5 = ScalarParameter() {defaultValue=1, group=HolographicTexture, parameterName=HologramOpacity, sortPriority=-10}
  n6 = Saturate(None=Add(A=NamedRerouteUsage(), B=ScalarParameter() {defaultValue=0.8, group=HolographicTexture, parameterName=Scanline_Opacity, sortPriority=0}) {constB=0.8})
  NamedRerouteDeclaration(None=LinearInterpolate(A=NamedRerouteUsage(), B=Multiply(A=@n5, B=@n6), Alpha=NamedRerouteUsage())) {name=HologramOpacity, nodeColor=<rgba 1 0.0938 0 1>}
  n7 = Saturate(None=Add(
    A=NamedRerouteUsage()
    B=ScalarParameter() {defaultValue=0.95, group=HolographicTexture, parameterName=Scanline_ColorMultiplier, sortPriority=0}
   ) {constB=0})
  n8 = VectorParameter() {desc=HOLOGRAM BASE COLOR, group=HolographicTexture, parameterName=HologramBaseColor, sortPriority=-4, defaultValue=<rgba 0.1357 0.2884 1 1>}
  n9 = TextureSampleParameter2D(UVs=NamedRerouteUsage()) {
   group=HolographicTexture
   parameterName=HologramTexture
   samplerSource=SSM_Clamp_WorldGroupSettings
   sortPriority=-20
   texture=/Game/2dAssets/Sprays/Season25/Textures/T-Sprays-Season25-S25-GalaxyCupReward
  }
  n10 = Multiply(A=FunctionInput(Preview=@n8.RGB) {bUsePreviewValueAsDefault=true, inputName=HologramBaseColor}, B=@n9.RGB)
  NamedRerouteDeclaration(None=LinearInterpolate(A=NamedRerouteUsage(), B=Multiply(A=@n7, B=@n10), Alpha=NamedRerouteUsage())) {name=HologramBaseColor, nodeColor=<rgba 0 0.2266 1 1>}
  n11 = TextureCoordinate()
  NamedRerouteDeclaration(None=MaterialFunctionCall!(Main_TextureUV=FunctionInput(Preview=@n11) {bUsePreviewValueAsDefault=true, inputName=TextureUV, inputType=FunctionInput_Vector2, sortPriority=2}) {materialFunction=/BRCosmetics/Characters/Player/Male/Medium/Bodies/M_MED_ByteGear/Materials/MF_ScreenGlitch_ByteGearRollGGWP}.Result) {name=ScreenGlitchUVs, nodeColor=<rgba 1 0 0.8672 1>}
  NamedRerouteDeclaration(None=@n9.RGBA) {name=Texture RGBA, nodeColor=<rgba 1 0 0.8906 1>}
  n12 = GetMaterialAttributes(None=FunctionInput() {inputName=MaterialAttributes, inputType=FunctionInput_MaterialAttributes, sortPriority=-1}) {attributeGetTypes [
    #3
    #4
    #1
    #2
   ]}
  NamedRerouteDeclaration(None=@n12.BaseColor) {name=OG Base Color, nodeColor=<rgba 0 0.7422 1 1>}
  NamedRerouteDeclaration(None=@n12.EmissiveColor) {name=OG Emissive, nodeColor=<rgba 0 1 0.1797 1>}
  NamedRerouteDeclaration(None=@n12.Opacity) {name=OG Opacity, nodeColor=<rgba 0.3422 0.8799 1 1>}
  NamedRerouteDeclaration(None=@n12.MaterialAttributes) {name=OG Material Attributes, nodeColor=<rgba 0.8984 0 1 1>}
  n13 = ChannelMaskParameter(None=NamedRerouteUsage()) {group=HolographicTexture, maskChannel=Alpha, parameterName=TextureMask_Channel, sortPriority=-10, defaultValue=<rgba 0 0 0 1>}
  n14 = TextureSampleParameter2D() {
   group=HolographicTexture
   parameterName=HologramMask
   sortPriority=-10
   texture=/BRCosmetics/Weapons/FORT_Melee/Pickaxe_CosmicSquatter/Textures/T_Pickaxe_CosmicSquatter_FX
  }
  n15 = FunctionInput(Preview=ChannelMaskParameter(None=@n14.RGB) {group=HolographicTexture, parameterName=HologramMask_Channel, sortPriority=-10}.R) {bUsePreviewValueAsDefault=true, inputName=HologramMask, inputType=FunctionInput_Scalar, sortPriority=5}
  n16 = Constant() {r=1}
  NamedRerouteDeclaration(None=StaticSwitchParameter(True=@n13.R, False=StaticSwitchParameter(True=@n15, False=@n16) {group=HolographicTexture, parameterName=UseHologramMask, sortPriority=-10}) {group=HolographicTexture, parameterName=UseTextureOnly, sortPriority=-5}) {name=TextureMask, nodeColor=<rgba 0.7344 0 1 1>}
  n17 = MaterialFunctionCall!(
   ScanLine_Tiling_1
   Scanline_PannerSpeed_1
   ScanLine_Tiling
   Scanline_PannerSpeed
   Scanline_Rotation
   UVs=FunctionInput(Preview=TextureCoordinate()) {bUsePreviewValueAsDefault=true, inputName=Scanlines_UV, inputType=FunctionInput_Vector2, sortPriority=2}
  ) {
   desc=ScanLine
   materialFunction=/BRCosmetics/Characters/Player/Male/Medium/Bodies/M_MED_ByteGear/Materials/MF_HolographicTexture_ByteGearRoll_ScanLines
  }
  NamedRerouteDeclaration(None=@n17.Result) {name=ScanLines, nodeColor=<rgba 0 1 0.4141 1>}
 ]}
