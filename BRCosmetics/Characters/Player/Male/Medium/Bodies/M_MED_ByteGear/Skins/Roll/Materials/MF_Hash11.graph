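# Material function that hashes one scalar into a pseudo-random 0–1 scalar, stored with the ByteGear Roll skin materials.
# The input is scaled by a small constant and wrapped with Frac, multiplied by an offset copy of itself, then by its own double, and wrapped again for output.
# The math matches Dave Hoskins' sine-free hash11, with named reroutes marking each step.
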MaterialFunction /BRCosmetics/Characters/Player/Male/Medium/Bodies/M_MED_ByteGear/Skins/Roll/Materials/MF_Hash11
graph {nodes [
  FunctionOutput(None=Frac(None=NamedRerouteUsage()))
  NamedRerouteDeclaration(None=FunctionInput() {bUsePreviewValueAsDefault=true, inputType=FunctionInput_Scalar, previewValue=<11.8304 0 0 1>}) {name=p, nodeColor=<rgba 1 0.2812 0 1>}
  n1 = NamedRerouteUsage()
  NamedRerouteDeclaration(None=Frac(None=Multiply(A=@n1) {constB=0.1031})) {name=p2, nodeColor=<rgba 0 0.0391 1 1>}
  n2 = NamedRerouteUsage()
  NamedRerouteDeclaration(None=Multiply(A=NamedRerouteUsage(), B=Add(A=@n2) {constB=33.33})) {name=p3, nodeColor=<rgba 0.3594 1 0 1>}
  n3 = NamedRerouteUsage()
  NamedRerouteDeclaration(None=Multiply(A=NamedRerouteUsage(), B=Add(A=@n3, B=@n3))) {name=p4, nodeColor=<rgba 1 0 0.6797 1>}
 ]}
