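# Material function that adds animated holographic glitch FX to the ByteGear Roll character skin.
# It passes incoming material attributes through MF_ByteGear_HolographicTexture, then adds glow to the resulting emissive color.
# The glow sums three time-scrolled, screen-glitch-distorted MF_GlitchPattern layers, each colored through curve atlas gradients.
# Helper branches build camera-facing virtual-plane, local planar and glitched screen UVs, plus offset RGB channel samples and a hash-jittered box mask.
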MaterialFunction /BRCosmetics/Characters/Player/Male/Medium/Bodies/M_MED_ByteGear/Skins/Roll/Materials/MF_ByteGear_Roll_FX
graph {expressions {
  MaterialExpressionAdd_0 {
   class=MaterialExpressionAdd
   inputs [
    A {from=MaterialExpressionScreenPosition_0, output=ViewportUV}
    B {from=MaterialExpressionMultiply_0}
   ]
  }
  MaterialExpressionAdd_1 {
   class=MaterialExpressionAdd
   inputs [
    A {from=MaterialExpressionNamedRerouteUsage_18}
    B {from=MaterialExpressionConvert_0}
   ]
  }
  MaterialExpressionAdd_10 {
   class=MaterialExpressionAdd
   inputs [
    A {from=MaterialExpressionAdd_9}
    B {from=MaterialExpressionMaterialFunctionCall_6, output=AlphaChannel}
   ]
  }
  MaterialExpressionAdd_11 {
   class=MaterialExpressionAdd
   inputs [
    A {from=MaterialExpressionAdd_22}
    B {from=MaterialExpressionConvert_1}
   ]
  }
  MaterialExpressionAdd_12 {
   class=MaterialExpressionAdd
   inputs [
    A {from=MaterialExpressionMaterialFunctionCall_11, output=RedChannel}
    B {from=MaterialExpressionMaterialFunctionCall_11, output=RedChannel}
   ]
  }
  MaterialExpressionAdd_13 {
   class=MaterialExpressionAdd
   inputs [
    A {from=MaterialExpressionAdd_12}
    B {from=MaterialExpressionMaterialFunctionCall_11, output=BlueChannel}
   ]
  }
  MaterialExpressionAdd_14 {
   class=MaterialExpressionAdd
   inputs [
    A {from=MaterialExpressionAdd_13}
    B {from=MaterialExpressionMaterialFunctionCall_11, output=AlphaChannel}
   ]
  }
  MaterialExpressionAdd_15 {
   class=MaterialExpressionAdd
   inputs [
    A {from=MaterialExpressionMultiply_6}
    B {from=MaterialExpressionMultiply_10}
   ]
  }
  MaterialExpressionAdd_16 {
   class=MaterialExpressionAdd
   inputs [
    A {from=MaterialExpressionAdd_21}
    B {from=MaterialExpressionConvert_2}
   ]
  }
  MaterialExpressionAdd_17 {
   class=MaterialExpressionAdd
   inputs [
    A {from=MaterialExpressionMaterialFunctionCall_21, output=RedChannel}
    B {from=MaterialExpressionMaterialFunctionCall_21, output=RedChannel}
   ]
  }
  MaterialExpressionAdd_18 {
   class=MaterialExpressionAdd
   inputs [
    A {from=MaterialExpressionAdd_17}
    B {from=MaterialExpressionMaterialFunctionCall_21, output=BlueChannel}
   ]
  }
  MaterialExpressionAdd_19 {
   class=MaterialExpressionAdd
   inputs [
    A {from=MaterialExpressionAdd_18}
    B {from=MaterialExpressionMaterialFunctionCall_21, output=AlphaChannel}
   ]
  }
  MaterialExpressionAdd_2 {
   class=MaterialExpressionAdd
   inputs [
    A {from=MaterialExpressionMaterialFunctionCall_6, output=RedChannel}
    B {from=MaterialExpressionMaterialFunctionCall_6, output=RedChannel}
   ]
  }
  MaterialExpressionAdd_20 {
   class=MaterialExpressionAdd
   inputs [
    A {from=MaterialExpressionAdd_15}
    B {from=MaterialExpressionMultiply_21}
   ]
  }
  MaterialExpressionAdd_21 {
   class=MaterialExpressionAdd
   inputs [
    A {from=MaterialExpressionNamedRerouteUsage_22}
    B {from=MaterialExpressionConstant2Vector_3}
   ]
  }
  MaterialExpressionAdd_22 {
   class=MaterialExpressionAdd
   inputs [
    A {from=MaterialExpressionNamedRerouteUsage_19}
    B {from=MaterialExpressionConstant2Vector_4}
   ]
  }
  MaterialExpressionAdd_3 {
   class=MaterialExpressionAdd
   inputs [
    A {from=MaterialExpressionMaterialFunctionCall_12, output=EmissiveColor}
    B {from=MaterialExpressionMultiply_1}
   ]
  }
  MaterialExpressionAdd_4 {
   class=MaterialExpressionAdd
   inputs [
    A {from=MaterialExpressionNamedRerouteUsage_14}
    B {from=""}
   ]
   properties {constB=0.002}
  }
  MaterialExpressionAdd_5 {
   class=MaterialExpressionAdd
   inputs [
    A {from=MaterialExpressionNamedRerouteUsage_15}
    B {from=""}
   ]
   properties {constB=-0.0031}
  }
  MaterialExpressionAdd_6 {
   class=MaterialExpressionAdd
   inputs [
    A {from=MaterialExpressionConstant2Vector_5}
    B {from=MaterialExpressionConstant2Vector_7}
   ]
  }
  MaterialExpressionAdd_7 {
   class=MaterialExpressionAdd
   inputs [
    A {from=MaterialExpressionNamedRerouteUsage_17}
    B {from=MaterialExpressionAppendVector_3}
   ]
   properties {constB=2}
  }
  MaterialExpressionAdd_8 {
   class=MaterialExpressionAdd
   inputs [
    A {from=MaterialExpressionMaterialFunctionCall_13, output=UVs}
    B {from=MaterialExpressionAppendVector_0}
   ]
  }
  MaterialExpressionAdd_9 {
   class=MaterialExpressionAdd
   inputs [
    A {from=MaterialExpressionAdd_2}
    B {from=MaterialExpressionMaterialFunctionCall_6, output=BlueChannel}
   ]
  }
  MaterialExpressionAppendVector_0 {
   class=MaterialExpressionAppendVector
   inputs [
    A {from=MaterialExpressionScalarParameter_11, output=""}
    B {from=MaterialExpressionScalarParameter_12, output=""}
   ]
  }
  MaterialExpressionAppendVector_1 {
   class=MaterialExpressionAppendVector
   inputs [
    A {from=MaterialExpressionScalarParameter_1, output=""}
    B {from=MaterialExpressionScalarParameter_2, output=""}
   ]
  }
  MaterialExpressionAppendVector_3 {
   class=MaterialExpressionAppendVector
   inputs [
    A {from=MaterialExpressionMultiply_15, output=""}
    B {from=MaterialExpressionSubtract_0, output=""}
   ]
  }
  MaterialExpressionCameraVectorWS_0 {class=MaterialExpressionCameraVectorWS, inputs [], properties {}}
  MaterialExpressionComponentMask_0 {class=MaterialExpressionComponentMask, inputs [None {from=MaterialExpressionMultiply_2}], properties {b=true, r=true}}
  MaterialExpressionComponentMask_1 {class=MaterialExpressionComponentMask, inputs [None {from=MaterialExpressionFunctionInput_1}], properties {g=true}}
  MaterialExpressionComponentMask_2 {class=MaterialExpressionComponentMask, inputs [None {from=MaterialExpressionSaturate_0}], properties {b=true, g=true, r=true}}
  MaterialExpressionComponentMask_4 {class=MaterialExpressionComponentMask, inputs [None {from=MaterialExpressionTextureCoordinate_7}], properties {g=true}}
  MaterialExpressionComponentMask_5 {class=MaterialExpressionComponentMask, inputs [None {from=MaterialExpressionSaturate_1}], properties {b=true, g=true, r=true}}
  MaterialExpressionComponentMask_7 {class=MaterialExpressionComponentMask, inputs [None {from=MaterialExpressionSaturate_2}], properties {b=true, g=true, r=true}}
  MaterialExpressionConstant2Vector_0 {class=MaterialExpressionConstant2Vector, properties {g=1, r=2}}
  MaterialExpressionConstant2Vector_1 {class=MaterialExpressionConstant2Vector, properties {g=0.9, r=1.1}}
  MaterialExpressionConstant2Vector_2 {class=MaterialExpressionConstant2Vector, properties {g=0.33, r=2.1}}
  MaterialExpressionConstant2Vector_3 {class=MaterialExpressionConstant2Vector, properties {g=0.67}}
  MaterialExpressionConstant2Vector_4 {class=MaterialExpressionConstant2Vector, properties {g=0.367, r=0.1}}
  MaterialExpressionConstant2Vector_5 {class=MaterialExpressionConstant2Vector, properties {g=0.5, r=0.8}}
  MaterialExpressionConstant2Vector_6 {class=MaterialExpressionConstant2Vector, properties {g=0.25, r=0.67}}
  MaterialExpressionConstant2Vector_7 {class=MaterialExpressionConstant2Vector, properties {g=2}}
  MaterialExpressionConstant3Vector_0 {class=MaterialExpressionConstant3Vector, properties {constant=<rgba 0 1 0 0>}}
  MaterialExpressionConstant3Vector_1 {class=MaterialExpressionConstant3Vector, properties {constant=<rgba 0 0 1 0>}}
  MaterialExpressionConstant_0 {class=MaterialExpressionConstant, properties {r=0.5}}
  MaterialExpressionConstant_1 {class=MaterialExpressionConstant}
  MaterialExpressionConstant_10 {class=MaterialExpressionConstant, properties {r=0.4}}
  MaterialExpressionConstant_11 {class=MaterialExpressionConstant, properties {r=0.4}}
  MaterialExpressionConstant_12 {class=MaterialExpressionConstant, properties {r=0.4}}
  MaterialExpressionConstant_13 {class=MaterialExpressionConstant, properties {r=0.25}}
  MaterialExpressionConstant_14 {class=MaterialExpressionConstant, properties {r=0.3}}
  MaterialExpressionConstant_15 {class=MaterialExpressionConstant, properties {r=0.3}}
  MaterialExpressionConstant_16 {class=MaterialExpressionConstant, properties {r=0.3}}
  MaterialExpressionConstant_2 {class=MaterialExpressionConstant, properties {r=0.5}}
  MaterialExpressionConstant_3 {class=MaterialExpressionConstant}
  MaterialExpressionConstant_4 {class=MaterialExpressionConstant, properties {r=3}}
  MaterialExpressionConstant_5 {class=MaterialExpressionConstant, properties {r=0.4}}
  MaterialExpressionConstant_6 {class=MaterialExpressionConstant, properties {r=0.4}}
  MaterialExpressionConstant_7 {class=MaterialExpressionConstant, properties {r=0.4}}
  MaterialExpressionConstant_8 {class=MaterialExpressionConstant, properties {r=0.25}}
  MaterialExpressionConstant_9 {class=MaterialExpressionConstant, properties {r=0.25}}
  MaterialExpressionConvert_0 {
   class=MaterialExpressionConvert
   inputs [
    None {from=MaterialExpressionMultiply_4}
    None {from=MaterialExpressionMultiply_5}
   ]
   properties {
    nodeName=Make Float2
    convertInputs [
     {
      type=Scalar
      expressionInput {
       expression=&self:MaterialExpressionMultiply_4
       inputName=None
       mask=0
       maskA=0
       maskB=0
       maskG=0
       maskR=0
       outputIndex=0
      }
     }
     {
      type=Scalar
      expressionInput {
       expression=&self:MaterialExpressionMultiply_5
       inputName=None
       mask=0
       maskA=0
       maskB=0
       maskG=0
       maskR=0
       outputIndex=0
      }
     }
    ]
    convertOutputs [{type=Vector2}]
   }
  }
  MaterialExpressionConvert_1 {
   class=MaterialExpressionConvert
   inputs [
    None {from=MaterialExpressionMultiply_8}
    None {from=MaterialExpressionMultiply_9}
   ]
   properties {
    nodeName=Make Float2
    convertInputs [
     {
      type=Scalar
      expressionInput {
       expression=&self:MaterialExpressionMultiply_8
       inputName=None
       mask=0
       maskA=0
       maskB=0
       maskG=0
       maskR=0
       outputIndex=0
      }
     }
     {
      type=Scalar
      expressionInput {
       expression=&self:MaterialExpressionMultiply_9
       inputName=None
       mask=0
       maskA=0
       maskB=0
       maskG=0
       maskR=0
       outputIndex=0
      }
     }
    ]
    convertOutputs [{type=Vector2}]
   }
  }
  MaterialExpressionConvert_2 {
   class=MaterialExpressionConvert
   inputs [
    None {from=MaterialExpressionMultiply_18}
    None {from=MaterialExpressionMultiply_19}
   ]
   properties {
    nodeName=Make Float2
    convertInputs [
     {
      type=Scalar
      expressionInput {
       expression=&self:MaterialExpressionMultiply_18
       inputName=None
       mask=0
       maskA=0
       maskB=0
       maskG=0
       maskR=0
       outputIndex=0
      }
     }
     {
      type=Scalar
      expressionInput {
       expression=&self:MaterialExpressionMultiply_19
       inputName=None
       mask=0
       maskA=0
       maskB=0
       maskG=0
       maskR=0
       outputIndex=0
      }
     }
    ]
    convertOutputs [{type=Vector2}]
   }
  }
  MaterialExpressionCrossProduct_0 {
   class=MaterialExpressionCrossProduct
   inputs [
    A {from=MaterialExpressionCameraVectorWS_0, output=""}
    B {from=MaterialExpressionTransform_0, output=""}
   ]
   properties {}
  }
  MaterialExpressionCurveAtlasRowParameter_0 {
   class=MaterialExpressionCurveAtlasRowParameter
   inputs [CurveTime {from=MaterialExpressionDesaturation_1, output=""}]
   properties {
    atlas=&AthenaCurveAtlas01
    curve=/Game/Athena/LinearColorCurves/Curve_Glider_Drift
    defaultValue=105
    parameterName=ColorCurve
   }
  }
  MaterialExpressionCurveAtlasRowParameter_1 {
   class=MaterialExpressionCurveAtlasRowParameter
   inputs [CurveTime {from=MaterialExpressionDesaturation_2, output=""}]
   properties {
    atlas=&AthenaCurveAtlas01
    curve=&Curve_Glider_Drift
    defaultValue=105
    parameterName=ColorCurve
   }
  }
  MaterialExpressionCurveAtlasRowParameter_2 {
   class=MaterialExpressionCurveAtlasRowParameter
   inputs [CurveTime {from=MaterialExpressionDesaturation_0, output=""}]
   properties {
    atlas=&AthenaCurveAtlas01
    curve=&Curve_Glider_Drift
    defaultValue=105
    parameterName=ColorCurve
   }
  }
  MaterialExpressionCurveAtlasRowParameter_3 {class=MaterialExpressionCurveAtlasRowParameter, inputs [CurveTime {from="", output=""}], properties {}}
  MaterialExpressionDesaturation_0 {
   class=MaterialExpressionDesaturation
   inputs [
    None {from=MaterialExpressionComponentMask_2, output=""}
    Fraction {from="", output=""}
   ]
   properties {}
  }
  MaterialExpressionDesaturation_1 {
   class=MaterialExpressionDesaturation
   inputs [
    None {from=MaterialExpressionComponentMask_5, output=""}
    Fraction {from="", output=""}
   ]
   properties {}
  }
  MaterialExpressionDesaturation_2 {
   class=MaterialExpressionDesaturation
   inputs [
    None {from=MaterialExpressionComponentMask_7, output=""}
    Fraction {from="", output=""}
   ]
   properties {}
  }
  MaterialExpressionFloor_0 {class=MaterialExpressionFloor, inputs [None {from=MaterialExpressionMultiply_13, output=""}], properties {}}
  MaterialExpressionFunctionInput_0 {
   class=MaterialExpressionFunctionInput
   inputs [Preview {from=MaterialExpressionSetMaterialAttributes_1, output=""}]
   properties {bUsePreviewValueAsDefault=true, inputName=In_MaterialAttributes, inputType=FunctionInput_MaterialAttributes}
  }
  MaterialExpressionFunctionInput_1 {class=MaterialExpressionFunctionInput, inputs [Preview {from="", output=""}], properties {inputName=TechArtMask, sortPriority=1}}
  MaterialExpressionFunctionOutput_0 {class=MaterialExpressionFunctionOutput, inputs [None {from=MaterialExpressionSetMaterialAttributes_0, output=""}]}
  MaterialExpressionLinearInterpolate_7 {
   class=MaterialExpressionLinearInterpolate
   inputs [
    A {from=MaterialExpressionVectorParameter_4, output=RGB}
    B {from=MaterialExpressionVectorParameter_5, output=RGB}
    Alpha {from=MaterialExpressionComponentMask_4}
   ]
  }
  MaterialExpressionLocalPosition_0 {class=MaterialExpressionLocalPosition, inputs [], properties {}}
  MaterialExpressionMaterialFunctionCall_0 {
   class=MaterialExpressionMaterialFunctionCall
   inputs [
    UVs {from=MaterialExpressionNamedRerouteUsage_16}
    Texture Scale {from=MaterialExpressionNamedRerouteUsage_9}
   ]
   properties {materialFunction=&ScaleUVsByCenter}
  }
  MaterialExpressionMaterialFunctionCall_1 {
   class=MaterialExpressionMaterialFunctionCall
   inputs [
    UVs {from=MaterialExpressionAdd_5}
    Texture Scale {from=MaterialExpressionNamedRerouteUsage_10}
   ]
   properties {materialFunction=&ScaleUVsByCenter}
  }
  MaterialExpressionMaterialFunctionCall_10 {
   class=MaterialExpressionMaterialFunctionCall
   inputs [
    A {from=MaterialExpressionAdd_7}
    B {from=MaterialExpressionAdd_6}
    Bounds {from=MaterialExpressionConstant2Vector_6}
    Edge Falloff {from=MaterialExpressionScalarParameter_13}
   ]
   properties {materialFunction=/Engine/Functions/Engine_MaterialFunctions01/Gradient/BoxMask-2D}
  }
  MaterialExpressionMaterialFunctionCall_11 {
   class=MaterialExpressionMaterialFunctionCall
   inputs [
    uv {from=MaterialExpressionMultiply_24}
    glitchtimemultiplier {from=MaterialExpressionScalarParameter_5}
    RedNoiseLowValue {from=MaterialExpressionConstant_10}
    GreenNoiseLowValue {from=MaterialExpressionConstant_11}
    BlueNoiseLowValue {from=MaterialExpressionConstant_12}
   ]
   properties {materialFunction=&/MF_GlitchPattern}
  }
  MaterialExpressionMaterialFunctionCall_12 {
   class=MaterialExpressionMaterialFunctionCall
   inputs [
    MaterialAttributes {from=MaterialExpressionNamedRerouteUsage_0}
    HologramBaseColor {from=MaterialExpressionNamedRerouteUsage_2}
    Scanlines_UV {from=MaterialExpressionNamedRerouteUsage_1}
    TextureUV {from=MaterialExpressionNamedRerouteUsage_3}
    HologramMask {from=MaterialExpressionNamedRerouteUsage_4}
   ]
   properties {materialFunction=&/MF_ByteGear_HolographicTexture}
  }
  MaterialExpressionMaterialFunctionCall_13 {
   class=MaterialExpressionMaterialFunctionCall
   inputs [
    UVs {from=MaterialExpressionTextureCoordinate_6}
    Texture Scale {from=MaterialExpressionScalarParameter_9}
   ]
   properties {materialFunction=&ScaleUVsByCenter}
  }
  MaterialExpressionMaterialFunctionCall_14 {
   class=MaterialExpressionMaterialFunctionCall
   inputs [
    UVs {from=MaterialExpressionMaterialFunctionCall_23, output=Result}
    Rotation Center {from=""}
    "Rotation Angle (0-1)" {from=MaterialExpressionConstant_13}
   ]
   properties {materialFunction=&CustomRotator}
  }
  MaterialExpressionMaterialFunctionCall_15 {
   class=MaterialExpressionMaterialFunctionCall
   inputs [Main_TextureUV {from=MaterialExpressionMaterialFunctionCall_0, output=UVs}]
   properties {materialFunction=/Game/Global/Materials/MaterialGym/Materials/ScreenGlitchGeneric/MF_ScreenGlitch_Generic}
  }
  MaterialExpressionMaterialFunctionCall_16 {
   class=MaterialExpressionMaterialFunctionCall
   inputs [Main_TextureUV {from=MaterialExpressionMaterialFunctionCall_1, output=UVs}]
   properties {materialFunction=&MF_ScreenGlitch_Generic}
  }
  MaterialExpressionMaterialFunctionCall_17 {
   class=MaterialExpressionMaterialFunctionCall
   inputs [Main_TextureUV {from=MaterialExpressionMaterialFunctionCall_2, output=UVs}]
   properties {materialFunction=&MF_ScreenGlitch_Generic}
  }
  MaterialExpressionMaterialFunctionCall_18 {
   class=MaterialExpressionMaterialFunctionCall
   inputs [Main_TextureUV {from=MaterialExpressionAdd_11}]
   properties {materialFunction=&MF_ScreenGlitch_Generic}
  }
  MaterialExpressionMaterialFunctionCall_19 {
   class=MaterialExpressionMaterialFunctionCall
   inputs [Main_TextureUV {from=MaterialExpressionAdd_1}]
   properties {materialFunction=&MF_ScreenGlitch_Generic}
  }
  MaterialExpressionMaterialFunctionCall_2 {
   class=MaterialExpressionMaterialFunctionCall
   inputs [
    UVs {from=MaterialExpressionAdd_4}
    Texture Scale {from=MaterialExpressionNamedRerouteUsage_11}
   ]
   properties {materialFunction=&ScaleUVsByCenter}
  }
  MaterialExpressionMaterialFunctionCall_20 {
   class=MaterialExpressionMaterialFunctionCall
   inputs [In {from=MaterialExpressionFloor_0}]
   properties {materialFunction=&/MF_Hash11}
  }
  MaterialExpressionMaterialFunctionCall_21 {
   class=MaterialExpressionMaterialFunctionCall
   inputs [
    uv {from=MaterialExpressionMultiply_23}
    glitchtimemultiplier {from=MaterialExpressionScalarParameter_6}
    RedNoiseLowValue {from=MaterialExpressionConstant_14}
    GreenNoiseLowValue {from=MaterialExpressionConstant_15}
    BlueNoiseLowValue {from=MaterialExpressionConstant_16}
   ]
   properties {materialFunction=&MF_GlitchPattern}
  }
  MaterialExpressionMaterialFunctionCall_23 {
   class=MaterialExpressionMaterialFunctionCall
   inputs [Main_TextureUV {from=MaterialExpressionAdd_16}]
   properties {materialFunction=&MF_ScreenGlitch_Generic}
  }
  MaterialExpressionMaterialFunctionCall_3 {
   class=MaterialExpressionMaterialFunctionCall
   inputs [
    R {from=MaterialExpressionTextureSample_1, output=R}
    G {from=MaterialExpressionTextureSample_2, output=G}
    B {from=MaterialExpressionTextureSample_3, output=B}
    A {from=""}
   ]
   properties {materialFunction=&AppendMany}
  }
  MaterialExpressionMaterialFunctionCall_4 {
   class=MaterialExpressionMaterialFunctionCall
   inputs [
    Plane Normal Axis {from=MaterialExpressionCameraVectorWS_0}
    Plane X Axis {from=MaterialExpressionNormalize_0}
    Plane Center {from=MaterialExpressionMaterialFunctionCall_5, output=Result}
    UV Scale {from=MaterialExpressionScalarParameter_7}
   ]
   properties {materialFunction=&VirtualPlaneCoordinates}
  }
  MaterialExpressionMaterialFunctionCall_5 {
   class=MaterialExpressionMaterialFunctionCall
   inputs []
   properties {materialFunction=&ComponentPivotLocation}
  }
  MaterialExpressionMaterialFunctionCall_6 {
   class=MaterialExpressionMaterialFunctionCall
   inputs [
    uv {from=MaterialExpressionMultiply_11}
    glitchtimemultiplier {from=MaterialExpressionScalarParameter_4}
    RedNoiseLowValue {from=MaterialExpressionConstant_5}
    GreenNoiseLowValue {from=MaterialExpressionConstant_6}
    BlueNoiseLowValue {from=MaterialExpressionConstant_7}
   ]
   properties {materialFunction=&MF_GlitchPattern}
  }
  MaterialExpressionMaterialFunctionCall_7 {
   class=MaterialExpressionMaterialFunctionCall
   inputs [
    UVs {from=MaterialExpressionMaterialFunctionCall_19, output=Result}
    Rotation Center {from=""}
    "Rotation Angle (0-1)" {from=MaterialExpressionConstant_9}
   ]
   properties {materialFunction=&CustomRotator}
  }
  MaterialExpressionMaterialFunctionCall_8 {
   class=MaterialExpressionMaterialFunctionCall
   inputs [
    UV {from=MaterialExpressionMaterialFunctionCall_4, output=UVs}
    Size {from=""}
    Curvature {from=""}
   ]
   properties {materialFunction=/Game/Accessories/FORT_Backpacks/Backpack_M_MED_PitGlass/FX/Materials/MF_Sphere}
  }
  MaterialExpressionMaterialFunctionCall_9 {
   class=MaterialExpressionMaterialFunctionCall
   inputs [
    UVs {from=MaterialExpressionMaterialFunctionCall_18, output=Result}
    Rotation Center {from=""}
    "Rotation Angle (0-1)" {from=MaterialExpressionConstant_8}
   ]
   properties {materialFunction=&CustomRotator}
  }
  MaterialExpressionMultiply_0 {
   class=MaterialExpressionMultiply
   inputs [
    A {from=MaterialExpressionTextureSample_0, output=R}
    B {from=MaterialExpressionScalarParameter_0}
   ]
  }
  MaterialExpressionMultiply_1 {
   class=MaterialExpressionMultiply
   inputs [
    A {from=MaterialExpressionAdd_20}
    B {from=MaterialExpressionNamedRerouteUsage_5}
   ]
  }
  MaterialExpressionMultiply_10 {
   class=MaterialExpressionMultiply
   inputs [
    A {from=MaterialExpressionMultiply_7}
    B {from=""}
   ]
   properties {constB=0.25}
  }
  MaterialExpressionMultiply_11 {
   class=MaterialExpressionMultiply
   inputs [
    A {from=MaterialExpressionMaterialFunctionCall_7, output=Rotated Values}
    B {from=MaterialExpressionConstant2Vector_0}
   ]
  }
  MaterialExpressionMultiply_12 {
   class=MaterialExpressionMultiply
   inputs [
    A {from=MaterialExpressionMultiply_3}
    B {from=MaterialExpressionMaterialFunctionCall_10, output=Result}
   ]
  }
  MaterialExpressionMultiply_13 {
   class=MaterialExpressionMultiply
   inputs [
    A {from=MaterialExpressionTime_2}
    B {from=""}
   ]
   properties {constB=20}
  }
  MaterialExpressionMultiply_14 {
   class=MaterialExpressionMultiply
   inputs [
    A {from=MaterialExpressionMaterialFunctionCall_20, output=Result}
    B {from=""}
   ]
   properties {constB=2}
  }
  MaterialExpressionMultiply_15 {
   class=MaterialExpressionMultiply
   inputs [
    A {from=MaterialExpressionSubtract_0}
    B {from=""}
   ]
   properties {constB=-1}
  }
  MaterialExpressionMultiply_16 {
   class=MaterialExpressionMultiply
   inputs [
    A {from=MaterialExpressionMultiply_1}
    B {from=MaterialExpressionNamedRerouteUsage_21}
   ]
  }
  MaterialExpressionMultiply_17 {
   class=MaterialExpressionMultiply
   inputs [
    A {from=MaterialExpressionComponentMask_7}
    B {from=MaterialExpressionCurveAtlasRowParameter_1}
   ]
  }
  MaterialExpressionMultiply_18 {
   class=MaterialExpressionMultiply
   inputs [
    A {from=MaterialExpressionTime_4}
    B {from=""}
   ]
   properties {constB=-0.011}
  }
  MaterialExpressionMultiply_19 {
   class=MaterialExpressionMultiply
   inputs [
    A {from=MaterialExpressionTime_4}
    B {from=""}
   ]
   properties {constB=-0.11}
  }
  MaterialExpressionMultiply_2 {
   class=MaterialExpressionMultiply
   inputs [
    A {from=MaterialExpressionLocalPosition_0, output=XYZ}
    B {from=MaterialExpressionScalarParameter_3}
   ]
  }
  MaterialExpressionMultiply_20 {
   class=MaterialExpressionMultiply
   inputs [
    A {from=MaterialExpressionWorldPosition_0, output=Z}
    B {from=MaterialExpressionScalarParameter_8}
   ]
   properties {constB=0.005}
  }
  MaterialExpressionMultiply_21 {
   class=MaterialExpressionMultiply
   inputs [
    A {from=MaterialExpressionMultiply_17}
    B {from=""}
   ]
   properties {constB=0.1}
  }
  MaterialExpressionMultiply_23 {
   class=MaterialExpressionMultiply
   inputs [
    A {from=MaterialExpressionMaterialFunctionCall_14, output=Rotated Values}
    B {from=MaterialExpressionConstant2Vector_2}
   ]
  }
  MaterialExpressionMultiply_24 {
   class=MaterialExpressionMultiply
   inputs [
    A {from=MaterialExpressionMaterialFunctionCall_9, output=Rotated Values}
    B {from=MaterialExpressionConstant2Vector_1}
   ]
  }
  MaterialExpressionMultiply_3 {
   class=MaterialExpressionMultiply
   inputs [
    A {from=MaterialExpressionComponentMask_2}
    B {from=MaterialExpressionCurveAtlasRowParameter_2}
   ]
  }
  MaterialExpressionMultiply_4 {
   class=MaterialExpressionMultiply
   inputs [
    A {from=MaterialExpressionTime_0}
    B {from=""}
   ]
   properties {constB=0.027}
  }
  MaterialExpressionMultiply_5 {
   class=MaterialExpressionMultiply
   inputs [
    A {from=MaterialExpressionTime_0}
    B {from=""}
   ]
   properties {constB=0.013}
  }
  MaterialExpressionMultiply_6 {
   class=MaterialExpressionMultiply
   inputs [
    A {from=MaterialExpressionMultiply_3}
    B {from=""}
   ]
   properties {constB=0.45}
  }
  MaterialExpressionMultiply_7 {
   class=MaterialExpressionMultiply
   inputs [
    A {from=MaterialExpressionComponentMask_5}
    B {from=MaterialExpressionCurveAtlasRowParameter_0}
   ]
  }
  MaterialExpressionMultiply_8 {
   class=MaterialExpressionMultiply
   inputs [
    A {from=MaterialExpressionTime_3}
    B {from=""}
   ]
   properties {constB=-0.021}
  }
  MaterialExpressionMultiply_9 {
   class=MaterialExpressionMultiply
   inputs [
    A {from=MaterialExpressionTime_3}
    B {from=""}
   ]
   properties {constB=0.041}
  }
  MaterialExpressionNamedRerouteDeclaration_0 {
   class=MaterialExpressionNamedRerouteDeclaration
   inputs [None {from=MaterialExpressionFunctionInput_0}]
   properties {name=In_MaterialAttributes, nodeColor=<rgba 0 1 0.3438 1>}
  }
  MaterialExpressionNamedRerouteDeclaration_1 {
   class=MaterialExpressionNamedRerouteDeclaration
   inputs [None {from=MaterialExpressionStaticSwitchParameter_8}]
   properties {name=WorldPositionOrTexCoord, nodeColor=<rgba 1 0.0234 0 1>}
  }
  MaterialExpressionNamedRerouteDeclaration_10 {
   class=MaterialExpressionNamedRerouteDeclaration
   inputs [None {from=MaterialExpressionComponentMask_0}]
   properties {name=PlanarUV, nodeColor=<rgba 0.6875 0 1 1>}
  }
  MaterialExpressionNamedRerouteDeclaration_11 {
   class=MaterialExpressionNamedRerouteDeclaration
   inputs [None {from=MaterialExpressionStaticSwitch_0}]
   properties {name=UVSwitch, nodeColor=<rgba 1 0.8438 0 1>}
  }
  MaterialExpressionNamedRerouteDeclaration_2 {
   class=MaterialExpressionNamedRerouteDeclaration
   inputs [None {from=MaterialExpressionStaticSwitchParameter_10}]
   properties {name=GradientBaseColor, nodeColor=<rgba 0 0.2969 1 1>}
  }
  MaterialExpressionNamedRerouteDeclaration_3 {
   class=MaterialExpressionNamedRerouteDeclaration
   inputs [None {from=MaterialExpressionStaticSwitchParameter_9}]
   properties {name=TextureUV, nodeColor=<rgba 0.6172 1 0 1>}
  }
  MaterialExpressionNamedRerouteDeclaration_4 {
   class=MaterialExpressionNamedRerouteDeclaration
   inputs [None {from=MaterialExpressionComponentMask_1}]
   properties {name=TechArtMask, nodeColor=<rgba 0 1 0.8125 1>}
  }
  MaterialExpressionNamedRerouteDeclaration_5 {
   class=MaterialExpressionNamedRerouteDeclaration
   inputs [None {from=MaterialExpressionAdd_0}]
   properties {name=GlitchUVs, nodeColor=<rgba 0 0.0391 1 1>}
  }
  MaterialExpressionNamedRerouteDeclaration_6 {
   class=MaterialExpressionNamedRerouteDeclaration
   inputs [None {from=MaterialExpressionMaterialFunctionCall_8, output=UV}]
   properties {name=VirtualPlaneUVs, nodeColor=<rgba 0 1 0.2031 1>}
  }
  MaterialExpressionNamedRerouteDeclaration_7 {
   class=MaterialExpressionNamedRerouteDeclaration
   inputs [None {from=MaterialExpressionAppendVector_1}]
   properties {name=GlitchUVScale, nodeColor=<rgba 0.3594 1 0 1>}
  }
  MaterialExpressionNamedRerouteDeclaration_8 {
   class=MaterialExpressionNamedRerouteDeclaration
   inputs [None {from=MaterialExpressionMaterialFunctionCall_12, output=GlitchFrequencyMask}]
   properties {name=GlitchFrequencyMask, nodeColor=<rgba 0 1 0.1094 1>}
  }
  MaterialExpressionNamedRerouteDeclaration_9 {
   class=MaterialExpressionNamedRerouteDeclaration
   inputs [None {from=MaterialExpressionMaterialFunctionCall_12, output=HolographicGlitchUVs}]
   properties {name=HolographicGlitchUVs, nodeColor=<rgba 0.0781 0 1 1>}
  }
  MaterialExpressionNamedRerouteUsage_0 {class=MaterialExpressionNamedRerouteUsage}
  MaterialExpressionNamedRerouteUsage_1 {class=MaterialExpressionNamedRerouteUsage}
  MaterialExpressionNamedRerouteUsage_10 {class=MaterialExpressionNamedRerouteUsage}
  MaterialExpressionNamedRerouteUsage_11 {class=MaterialExpressionNamedRerouteUsage}
  MaterialExpressionNamedRerouteUsage_12 {class=MaterialExpressionNamedRerouteUsage}
  MaterialExpressionNamedRerouteUsage_13 {class=MaterialExpressionNamedRerouteUsage}
  MaterialExpressionNamedRerouteUsage_14 {class=MaterialExpressionNamedRerouteUsage}
  MaterialExpressionNamedRerouteUsage_15 {class=MaterialExpressionNamedRerouteUsage}
  MaterialExpressionNamedRerouteUsage_16 {class=MaterialExpressionNamedRerouteUsage}
  MaterialExpressionNamedRerouteUsage_17 {class=MaterialExpressionNamedRerouteUsage}
  MaterialExpressionNamedRerouteUsage_18 {class=MaterialExpressionNamedRerouteUsage}
  MaterialExpressionNamedRerouteUsage_19 {class=MaterialExpressionNamedRerouteUsage}
  MaterialExpressionNamedRerouteUsage_2 {class=MaterialExpressionNamedRerouteUsage}
  MaterialExpressionNamedRerouteUsage_20 {class=MaterialExpressionNamedRerouteUsage}
  MaterialExpressionNamedRerouteUsage_21 {class=MaterialExpressionNamedRerouteUsage}
  MaterialExpressionNamedRerouteUsage_22 {class=MaterialExpressionNamedRerouteUsage}
  MaterialExpressionNamedRerouteUsage_3 {class=MaterialExpressionNamedRerouteUsage}
  MaterialExpressionNamedRerouteUsage_4 {class=MaterialExpressionNamedRerouteUsage}
  MaterialExpressionNamedRerouteUsage_5 {class=MaterialExpressionNamedRerouteUsage}
  MaterialExpressionNamedRerouteUsage_6 {class=MaterialExpressionNamedRerouteUsage}
  MaterialExpressionNamedRerouteUsage_7 {class=MaterialExpressionNamedRerouteUsage}
  MaterialExpressionNamedRerouteUsage_8 {class=MaterialExpressionNamedRerouteUsage}
  MaterialExpressionNamedRerouteUsage_9 {class=MaterialExpressionNamedRerouteUsage}
  MaterialExpressionNormalize_0 {class=MaterialExpressionNormalize, inputs [VectorInput {from=MaterialExpressionCrossProduct_0, output=""}]}
  MaterialExpressionRotator_1 {
   class=MaterialExpressionRotator
   inputs [
    Coordinate {from=MaterialExpressionAdd_8, output=""}
    Time {from=MaterialExpressionScalarParameter_10, output=""}
   ]
   properties {}
  }
  MaterialExpressionSaturate_0 {class=MaterialExpressionSaturate, inputs [None {from=MaterialExpressionAdd_10}]}
  MaterialExpressionSaturate_1 {class=MaterialExpressionSaturate, inputs [None {from=MaterialExpressionAdd_14}]}
  MaterialExpressionSaturate_2 {class=MaterialExpressionSaturate, inputs [None {from=MaterialExpressionAdd_19}]}
  MaterialExpressionScalarParameter_0 {class=MaterialExpressionScalarParameter, properties {defaultValue=0.05, parameterName=GlitchTextureAmount}}
  MaterialExpressionScalarParameter_1 {class=MaterialExpressionScalarParameter, properties {defaultValue=0.67, parameterName=GlitchUVScaleU}}
  MaterialExpressionScalarParameter_10 {class=MaterialExpressionScalarParameter, properties {group=00_ExampleParameters, parameterName=TextureRotation, sortPriority=1}}
  MaterialExpressionScalarParameter_11 {class=MaterialExpressionScalarParameter, properties {group=00_ExampleParameters, parameterName=TexturePositionOffset_X, sortPriority=1}}
  MaterialExpressionScalarParameter_12 {class=MaterialExpressionScalarParameter, properties {group=00_ExampleParameters, parameterName=TexturePositionOffset_Y, sortPriority=1}}
  MaterialExpressionScalarParameter_13 {class=MaterialExpressionScalarParameter, properties {defaultValue=0.15, parameterName=.2}}
  MaterialExpressionScalarParameter_2 {class=MaterialExpressionScalarParameter, properties {defaultValue=0.67, parameterName=GlitchUVScaleV}}
  MaterialExpressionScalarParameter_3 {class=MaterialExpressionScalarParameter, properties {defaultValue=0.0333, parameterName=PlanarUVScale}}
  MaterialExpressionScalarParameter_4 {class=MaterialExpressionScalarParameter, properties {defaultValue=21, parameterName="glitchtimemultiplier (S)"}}
  MaterialExpressionScalarParameter_5 {class=MaterialExpressionScalarParameter, properties {defaultValue=34, parameterName="glitchtimemultiplier (S)"}}
  MaterialExpressionScalarParameter_6 {class=MaterialExpressionScalarParameter, properties {defaultValue=34, parameterName="glitchtimemultiplier (S)"}}
  MaterialExpressionScalarParameter_7 {class=MaterialExpressionScalarParameter, properties {defaultValue=50, parameterName=UV Scale}}
  MaterialExpressionScalarParameter_8 {class=MaterialExpressionScalarParameter, properties {defaultValue=0.005, group=00_ExampleParameters, parameterName=ScanlineTiling}}
  MaterialExpressionScalarParameter_9 {class=MaterialExpressionScalarParameter, properties {defaultValue=1, group=00_ExampleParameters, parameterName=TextureScale, sortPriority=1}}
  MaterialExpressionScreenPosition_0 {class=MaterialExpressionScreenPosition, inputs [], properties {}}
  MaterialExpressionSetMaterialAttributes_0 {
   class=MaterialExpressionSetMaterialAttributes
   inputs [
    MaterialAttributes {from=MaterialExpressionMaterialFunctionCall_12, output=MaterialAttributes}
    EmissiveColor {from=MaterialExpressionAdd_3, output=""}
    Opacity {from="", output=""}
    BaseColor {from="", output=""}
    WorldPositionOffset {from="", output=""}
   ]
   properties {attributeSetTypes [
     #1
     #2
     #3
     #4
    ]}
  }
  MaterialExpressionSetMaterialAttributes_1 {
   class=MaterialExpressionSetMaterialAttributes
   inputs [
    MaterialAttributes {from="", output=""}
    BaseColor {from=MaterialExpressionTextureSampleParameter2D_2, output=RGB}
    Roughness {from=MaterialExpressionConstant_0, output=""}
    Metallic {from=MaterialExpressionConstant_1, output=""}
    Specular {from=MaterialExpressionConstant_2, output=""}
    EmissiveColor {from=MaterialExpressionConstant_3, output=""}
    Normal {from=MaterialExpressionConstant3Vector_1, output=""}
   ]
   properties {attributeSetTypes [
     #3
     #5
     #6
     #7
     #1
     #8
    ]}
  }
  MaterialExpressionStaticSwitchParameter_10 {
   class=MaterialExpressionStaticSwitchParameter
   inputs [
    True {from=MaterialExpressionLinearInterpolate_7, output=""}
    False {from=MaterialExpressionVectorParameter_0, output=RGB}
   ]
   properties {group=00_ExampleParameters, parameterName=IsPlaneExample}
  }
  MaterialExpressionStaticSwitchParameter_8 {
   class=MaterialExpressionStaticSwitchParameter
   inputs [
    True {from=MaterialExpressionMultiply_20, output=""}
    False {from=MaterialExpressionTextureCoordinate_5, output=""}
   ]
   properties {group=00_ExampleParameters, parameterName=UseWorldPositionForScanlines}
  }
  MaterialExpressionStaticSwitchParameter_9 {
   class=MaterialExpressionStaticSwitchParameter
   inputs [
    True {from=MaterialExpressionRotator_1, output=""}
    False {from=MaterialExpressionTextureCoordinate_6, output=""}
   ]
   properties {group=00_ExampleParameters, parameterName="OffsetTexturePosition&&Scale&&Rotation"}
  }
  MaterialExpressionStaticSwitch_0 {
   class=MaterialExpressionStaticSwitch
   inputs [
    True {from=MaterialExpressionNamedRerouteUsage_12, output=""}
    False {from=MaterialExpressionNamedRerouteUsage_13, output=""}
    Value {from="", output=""}
   ]
   properties {}
  }
  MaterialExpressionSubtract_0 {
   class=MaterialExpressionSubtract
   inputs [
    A {from=MaterialExpressionMultiply_14}
    B {from=""}
   ]
  }
  MaterialExpressionTextureCoordinate_5 {class=MaterialExpressionTextureCoordinate}
  MaterialExpressionTextureCoordinate_6 {class=MaterialExpressionTextureCoordinate}
  MaterialExpressionTextureCoordinate_7 {class=MaterialExpressionTextureCoordinate}
  MaterialExpressionTextureObjectParameter_0 {
   class=MaterialExpressionTextureObjectParameter
   inputs [
    UVs {from="", output=""}
    Apply View MipBias {from="", output=""}
   ]
   properties {
    parameterName=GlitchUVTexture
    samplerType=SAMPLERTYPE_LinearColor
    texture=/BRCosmetics/Characters/Player/Male/Medium/Bodies/M_MED_ByteGear/Skins/Roll/Textures/T_ByteGear_Roll_Glitch_FX
   }
  }
  MaterialExpressionTextureObjectParameter_1 {
   class=MaterialExpressionTextureObjectParameter
   inputs [
    UVs {from="", output=""}
    Apply View MipBias {from="", output=""}
   ]
   properties {
    parameterName=GlitchUVTexture
    samplerType=SAMPLERTYPE_LinearColor
    texture=&T_ByteGear_Roll_Glitch_FX
   }
  }
  MaterialExpressionTextureObjectParameter_2 {
   class=MaterialExpressionTextureObjectParameter
   inputs [
    UVs {from="", output=""}
    Apply View MipBias {from="", output=""}
   ]
   properties {
    parameterName=GlitchUVTexture
    samplerType=SAMPLERTYPE_LinearColor
    texture=&T_ByteGear_Roll_Glitch_FX
   }
  }
  MaterialExpressionTextureObjectParameter_3 {
   class=MaterialExpressionTextureObjectParameter
   inputs [
    UVs {from="", output=""}
    Apply View MipBias {from="", output=""}
   ]
   properties {
    parameterName=GlitchUVTexture
    samplerType=SAMPLERTYPE_LinearColor
    texture=&T_ByteGear_Roll_Glitch_FX
   }
  }
  MaterialExpressionTextureSampleParameter2D_2 {
   class=MaterialExpressionTextureSampleParameter2D
   inputs [
    UVs {from="", output=""}
    Apply View MipBias {from="", output=""}
   ]
   properties {
    group=00_ExampleParameters
    parameterName=BaseColorTexture
    texture=/BRCosmetics/Accessories/FORT_Backpacks/Backpack_F_MED_ShinyStar/Textures/T_F_MED_ShinyStar_Backpack_D
   }
  }
  MaterialExpressionTextureSample_0 {
   class=MaterialExpressionTextureSample
   inputs [
    UVs {from=""}
    Tex {from=MaterialExpressionTextureObjectParameter_0}
    Apply View MipBias {from=""}
   ]
   properties {
    samplerType=SAMPLERTYPE_LinearColor
    texture=&T_ByteGear_Roll_Glitch_FX
   }
  }
  MaterialExpressionTextureSample_1 {
   class=MaterialExpressionTextureSample
   inputs [
    UVs {from=MaterialExpressionMaterialFunctionCall_15, output=Result}
    Tex {from=MaterialExpressionTextureObjectParameter_1}
    Apply View MipBias {from=""}
   ]
   properties {
    samplerSource=SSM_Wrap_WorldGroupSettings
    samplerType=SAMPLERTYPE_LinearColor
    texture=&T_ByteGear_Roll_Glitch_FX
   }
  }
  MaterialExpressionTextureSample_2 {
   class=MaterialExpressionTextureSample
   inputs [
    UVs {from=MaterialExpressionMaterialFunctionCall_16, output=Result}
    Tex {from=MaterialExpressionTextureObjectParameter_2}
    Apply View MipBias {from=""}
   ]
   properties {
    samplerSource=SSM_Wrap_WorldGroupSettings
    samplerType=SAMPLERTYPE_LinearColor
    texture=&T_ByteGear_Roll_Glitch_FX
   }
  }
  MaterialExpressionTextureSample_3 {
   class=MaterialExpressionTextureSample
   inputs [
    UVs {from=MaterialExpressionMaterialFunctionCall_17, output=Result}
    Tex {from=MaterialExpressionTextureObjectParameter_3}
    Apply View MipBias {from=""}
   ]
   properties {
    samplerSource=SSM_Wrap_WorldGroupSettings
    samplerType=SAMPLERTYPE_LinearColor
    texture=&T_ByteGear_Roll_Glitch_FX
   }
  }
  MaterialExpressionTime_0 {class=MaterialExpressionTime}
  MaterialExpressionTime_2 {class=MaterialExpressionTime, properties {bOverride_Period=true, period=30}}
  MaterialExpressionTime_3 {class=MaterialExpressionTime}
  MaterialExpressionTime_4 {class=MaterialExpressionTime}
  MaterialExpressionTransform_0 {
   class=MaterialExpressionTransform
   inputs [None {from=MaterialExpressionConstant3Vector_0}]
   properties {transformSourceType=TRANSFORMSOURCE_View}
  }
  MaterialExpressionVectorParameter_0 {
   class=MaterialExpressionVectorParameter
   properties {desc=HOLOGRAM BASE COLOR, group=HolographicTexture, parameterName=HologramBaseColor, sortPriority=-4, defaultValue=<rgba 0.1357 0.2884 1 1>}
  }
  MaterialExpressionVectorParameter_4 {
   class=MaterialExpressionVectorParameter
   properties {group=00_ExampleParameters, parameterName=GradientColor01, defaultValue=<rgba 1 0 0.6293 1>}
  }
  MaterialExpressionVectorParameter_5 {
   class=MaterialExpressionVectorParameter
   properties {group=00_ExampleParameters, parameterName=GradientColor02, defaultValue=<rgba 0 0.3224 1 1>}
  }
  MaterialExpressionWorldPosition_0 {class=MaterialExpressionWorldPosition, inputs [], properties {}}
  MaterialExpressionWorldPosition_1 {class=MaterialExpressionWorldPosition, inputs [], properties {}}
 }}
