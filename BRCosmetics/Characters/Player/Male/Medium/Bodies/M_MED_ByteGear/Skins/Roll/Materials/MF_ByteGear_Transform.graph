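# Material function in the ByteGear Roll skin folder, exposed as "Transform (UV)" under Motion Design and Material Designer, that transforms texture UVs.
# It rotates UVs around a pivot via CustomRotator, converting degrees to negated turns, then multiplies by scale about that pivot and adds a position offset.
# A nested function call applies horizontal and vertical slant.
# Effect Amount, a saturated percentage, blends from original to transformed UVs.
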MaterialFunction /BRCosmetics/Characters/Player/Male/Medium/Bodies/M_MED_ByteGear/Skins/Roll/Materials/MF_ByteGear_Transform
graph {
 properties {bExposeToLibrary=true, userExposedCaption="Transform (UV)", libraryCategoriesText [Motion Design, Material Designer]}
 nodes [
  n1 = FunctionInput(Preview=TextureCoordinate()) {inputName=UV, inputType=FunctionInput_Vector2}
  n2 = FunctionInput() {bUsePreviewValueAsDefault=true, inputName=Pivot, inputType=FunctionInput_Vector2, sortPriority=8, previewValue=<0.5 0.5 0 1>}
  n3 = FunctionInput() {
   bUsePreviewValueAsDefault=true
   desc="UIMin=0, UIMax=360, ClampMin=0, ClampMax=360"
   inputName=Rotation
   inputType=FunctionInput_Scalar
   sortPriority=5
  }
  n4 = Subtract(
   A=MaterialFunctionCall(UVs=@n1, Rotation Center=@n2, "Rotation Angle (0-1)"=Divide(A=@n3) {constB=-360}) {materialFunction=&CustomRotator}.Rotated Values
   B=@n2
  )
  n5 = FunctionInput() {
   bUsePreviewValueAsDefault=true
   desc="UIMin=0, UIMax=1, ClampMin=0, ClampMax=1"
   inputName=Scale
   inputType=FunctionInput_Vector2
   sortPriority=6
   previewValue=<1 1 0 1>
  }
  n6 = Add(
   A=Add(A=Multiply(A=@n4, B=@n5), B=@n2)
   B=FunctionInput() {bUsePreviewValueAsDefault=true, inputName=Position, inputType=FunctionInput_Vector2, sortPriority=1}
  )
  n7 = FunctionInput() {
   bUsePreviewValueAsDefault=true
   desc="UIMin=0, UIMax=1, ClampMin=0, ClampMax=1"
   inputName=Slant Horizontal
   inputType=FunctionInput_Scalar
   sortPriority=12
  }
  n8 = FunctionInput() {
   bUsePreviewValueAsDefault=true
   desc="UIMin=0, UIMax=1, ClampMin=0, ClampMax=1"
   inputName=Slant Vertical
   inputType=FunctionInput_Scalar
   sortPriority=14
  }
  n9 = FunctionInput() {
   bUsePreviewValueAsDefault=true
   desc="UIMin=0, UIMax=100, ClampMin=0, ClampMax=100"
   inputName=Effect Amount
   inputType=FunctionInput_Scalar
   sortPriority=99
   previewValue=<100 0 0 1>
  }
  n10 = Divide(A=@n9) {constB=100}
  FunctionOutput(None=LinearInterpolate(
    A=NamedRerouteUsage()
    B=MaterialFunctionCall!(UVs=@n6, Slant Horizontal=@n7, Slant Vertical=@n8, Effect Amount).Slanted UVs
    Alpha=Saturate(None=@n10)
   )) {description=DMB_TextureUV}
  NamedRerouteDeclaration(None=@n1) {name=UV, nodeColor=<rgba 0 0.7656 1 1>}
 ]
}
